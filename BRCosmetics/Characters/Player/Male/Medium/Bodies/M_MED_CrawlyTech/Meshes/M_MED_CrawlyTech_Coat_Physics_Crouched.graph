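# Physics asset for the CrawlyTech male medium character's coat in its crouched variant, simulating the jacket panels and a phone.
# Seven five-bone capsule chains (one box) hang from pelvis-constrained base bones with tight swing and twist limits.
# Cross constraints with free swing tie neighboring back, mid and front chains together like a cloth lattice.
# Pelvis, spine, leg and foot capsules supply body collision, and a wide flat root box forms a floor.
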PhysicsAsset /BRCosmetics/Characters/Player/Male/Medium/Bodies/M_MED_CrawlyTech/Meshes/M_MED_CrawlyTech_Coat_Physics_Crouched
bodies {
 base_backJacket [{shapeType=Capsule, length=0.505, radius=0.505, shapeName=base_backJacket_capsule, rotation=<rot 0 -90 0>}]
 base_backJacket_l [{shapeType=Capsule, length=0.505, radius=0.505, shapeName=base_backJacket_l_capsule, rotation=<rot 0 -90 0>}]
 base_backJacket_r [{shapeType=Capsule, length=0.505, radius=0.505, shapeName=base_backJacket_r_capsule, rotation=<rot 0 -90 0>}]
 base_frontJacket_l [{shapeType=Capsule, length=0.505, radius=0.505, shapeName=base_frontJacket_l_capsule, rotation=<rot 0 -90 0>}]
 base_frontJacket_r [{shapeType=Capsule, length=0.505, radius=0.505, shapeName=base_frontJacket_r_capsule, rotation=<rot 0 -90 0>}]
 base_midJacket_l [{shapeType=Capsule, length=0.505, radius=0.505, shapeName=base_midJacket_l_capsule, rotation=<rot 0 -90 0>}]
 base_midJacket_r [{shapeType=Capsule, length=0.505, radius=0.505, shapeName=base_midJacket_r_capsule, rotation=<rot 0 -90 0>}]
 calf_twist_02_l [{shapeType=Capsule, length=39.3994, radius=8.6006, shapeName=calf_l_capsule, center=<5.297 -2.8249 0.4076>, rotation=<rot 90 -2.1515 0>}]
 calf_twist_02_r [{shapeType=Capsule, length=39.309, radius=8.691, shapeName=calf_r_capsule, center=<-0.0005 2.626 -0.4076>, rotation=<rot 90 177.8486 0>}]
 dyn_backJacket_00 [{shapeType=Capsule, length=6, radius=3, shapeName=dyn_backJacket_00_capsule, center=<2 0 0>, rotation=<rot 0 0 90>}]
 dyn_backJacket_01 [{shapeType=Capsule, length=5, radius=4.5, shapeName=dyn_backJacket_01_capsule, center=<2 0 0.4926>, rotation=<rot 0 0 90>}]
 dyn_backJacket_02 [{shapeType=Capsule, length=8, radius=6, shapeName=dyn_backJacket_02_capsule, center=<2 0 8.4125>, rotation=<rot 0 0 90>}]
 dyn_backJacket_03 [{shapeType=Capsule, length=24.0826, radius=9.9683, shapeName=dyn_backJacket_03_capsule, center=<4.6193 0 5.258>, rotation=<rot 0 0 90>}]
 dyn_backJacket_04 [{shapeType=Capsule, length=23.0263, radius=9.9683, shapeName=dyn_backJacket_04_capsule, center=<10.576 0 4.4182>, rotation=<rot 0 0 90>}]
 dyn_backJacket_l_00 [{shapeType=Capsule, length=6, radius=3, shapeName=dyn_backJacket_l_00_capsule, center=<2 0 0>, rotation=<rot 0 0 90>}]
 dyn_backJacket_l_01 [{shapeType=Capsule, length=5, radius=4.5, shapeName=dyn_backJacket_l_01_capsule, center=<2 0 0>, rotation=<rot 0 0 90>}]
 dyn_backJacket_l_02 [{shapeType=Capsule, length=5.8826, radius=9.9683, shapeName=dyn_backJacket_l_02_capsule, center=<2 0 6.5>, rotation=<rot 0 0 90>}]
 dyn_backJacket_l_03 [{shapeType=Capsule, length=17.9202, radius=9.9683, shapeName=dyn_backJacket_l_03_capsule, center=<4.6193 0 5>, rotation=<rot 0 0 90>}]
 dyn_backJacket_l_04 [{shapeType=Capsule, length=8.1376, radius=9.9683, shapeName=dyn_backJacket_l_04_capsule, center=<10.576 0 5>, rotation=<rot 0 0 90>}]
 dyn_backJacket_r_00 [{shapeType=Capsule, length=6, radius=3, shapeName=dyn_backJacket_r_00_capsule, center=<2 0 0>, rotation=<rot 0 0 90>}]
 dyn_backJacket_r_01 [{shapeType=Capsule, length=5, radius=4.5, shapeName=dyn_backJacket_r_01_capsule, center=<2 0 0>, rotation=<rot 0 0 90>}]
 dyn_backJacket_r_02 [{shapeType=Capsule, length=5.4994, radius=10.3516, shapeName=dyn_backJacket_r_02_capsule, center=<2 0 6.5>, rotation=<rot 0 0 90>}]
 dyn_backJacket_r_03 [{shapeType=Capsule, length=17.9202, radius=9.9683, shapeName=dyn_backJacket_r_03_capsule, center=<4.6193 0 5>, rotation=<rot 0 0 90>}]
 dyn_backJacket_r_04 [{shapeType=Capsule, length=6.768, radius=9.9683, shapeName=dyn_backJacket_r_04_capsule, center=<8.0562 0 5>, rotation=<rot 0 0 90>}]
 dyn_frontJacket_l_00 [{shapeType=Capsule, length=3.2155, radius=4.9192, shapeName=dyn_frontJacket_l_00_capsule, center=<2 0 1.6207>, rotation=<rot 0 0 90>}]
 dyn_frontJacket_l_01 [{
   shapeType=Capsule
   length=7.2246
   radius=7.391
   shapeName=dyn_frontJacket_r_02_capsule
   center=<6.8596 0.1215 5.3722>
   rotation=<rot 0 0 84.4257>
  }]
 dyn_frontJacket_l_02 [{
   shapeType=Box
   extentX=13.0751
   extentY=11.2359
   extentZ=4.5263
   shapeName=dyn_frontJacket_l_02_box
   center=<7.7738 0 1.282>
   rotation=<rot -2.994 0 0>
  }]
 dyn_frontJacket_l_03 [{
   shapeType=Capsule
   length=8.2709
   radius=11.1516
   shapeName=dyn_frontJacket_r_03_capsule
   center=<8.9359 -3.4645 9.5282>
   rotation=<rot 0.0001 0 85.3831>
  }]
 dyn_frontJacket_l_04 [{
   shapeType=Capsule
   length=6
   radius=8
   shapeName=dyn_frontJacket_r_04_capsule
   center=<15 0.3302 7.1201>
   rotation=<rot 0.0001 -0.0001 77.07>
  }]
 dyn_frontJacket_r_00 [{shapeType=Capsule, length=3, radius=3, shapeName=dyn_frontJacket_r_00_capsule, center=<2 0 0>, rotation=<rot 0 0 90>}]
 dyn_frontJacket_r_01 [{shapeType=Capsule, length=3, radius=3, shapeName=dyn_frontJacket_r_01_capsule, center=<2 0 3.7495>, rotation=<rot 0 0 90>}]
 dyn_frontJacket_r_02 [{shapeType=Capsule, length=6, radius=6, shapeName=dyn_frontJacket_r_02_capsule, center=<9.1634 0 7.5892>, rotation=<rot 0 0 90>}]
 dyn_frontJacket_r_03 [{
   shapeType=Capsule
   length=8
   radius=6
   shapeName=dyn_frontJacket_r_03_capsule
   center=<12.2227 1.7139 9.1055>
   rotation=<rot 0.0001 0 85.3831>
  }]
 dyn_frontJacket_r_04 [{
   shapeType=Capsule
   length=8
   radius=6
   shapeName=dyn_frontJacket_r_04_capsule
   center=<12.9359 0.0663 8.2258>
   rotation=<rot 0.0001 -0.0001 77.07>
  }]
 dyn_midJacket_l_00 [{shapeType=Capsule, length=6, radius=3, shapeName=dyn_midJacket_l_00_capsule, center=<2 0 0>, rotation=<rot 0 0 90>}]
 dyn_midJacket_l_01 [{shapeType=Capsule, length=13.7619, radius=8.794, shapeName=dyn_midJacket_l_01_capsule, center=<3.983 0 6.686>, rotation=<rot 0 0 90>}]
 dyn_midJacket_l_02 [{shapeType=Capsule, length=6, radius=8, shapeName=dyn_midJacket_l_02_capsule, center=<2 0 5.867>, rotation=<rot 0 0 90>}]
 dyn_midJacket_l_03 [{shapeType=Capsule, length=6, radius=8, shapeName=dyn_midJacket_l_03_capsule, center=<11.2323 0 8.514>, rotation=<rot 0 0 90>}]
 dyn_midJacket_l_04 [{shapeType=Capsule, length=6, radius=8, shapeName=dyn_midJacket_r_04_capsule, center=<17.189 0 8.9075>, rotation=<rot 0 0 90>}]
 dyn_midJacket_r_00 [{shapeType=Capsule, length=6, radius=3, shapeName=dyn_midJacket_r_00_capsule, center=<2 0 0>, rotation=<rot 0 0 90>}]
 dyn_midJacket_r_01 [{shapeType=Capsule, length=3, radius=3, shapeName=dyn_midJacket_r_01_capsule, center=<2 0 3.3373>, rotation=<rot 0 0 90>}]
 dyn_midJacket_r_02 [{shapeType=Capsule, length=4, radius=3, shapeName=dyn_midJacket_r_02_capsule, center=<2 0 6.9248>, rotation=<rot 0 0 90>}]
 dyn_midJacket_r_03 [{shapeType=Capsule, length=6, radius=6, shapeName=dyn_midJacket_l_03_capsule, center=<8.8051 0 6.1226>, rotation=<rot 0 0 90>}]
 dyn_midJacket_r_04 [{shapeType=Capsule, length=6, radius=6, shapeName=dyn_midJacket_r_04_capsule, center=<13.4265 0 6.5162>, rotation=<rot 0 0 90>}]
 dyn_phone [{
   shapeType=Capsule
   length=8.5274
   radius=2.787
   shapeName=dyn_phone_capsule
   center=<1.968 0.4646 -0.4606>
   rotation=<rot 1.0859 -90 -1.2902>
  }]
 foot_l [{shapeType=Capsule, length=11.4706, radius=14.2223, shapeName=foot_l_capsule, center=<-5.5858 6.5454 -0.0611>, rotation=<rot 0 0 90>}]
 foot_r [{shapeType=Capsule, length=11.4706, radius=14.2223, shapeName=foot_r_capsule, center=<5.5858 -6.5454 0.0611>, rotation=<rot 0 0 -90>}]
 pelvis [{shapeType=Capsule, length=0.9078, radius=10.7288, shapeName=pelvis_capsule, center=<-7.3953 1.0001 0>}]
 root [{shapeType=Box, extentX=273.7882, extentY=244.2862, extentZ=7.841, shapeName=root_box, center=<0 0 -4.4106>}]
 spine_01 [{shapeType=Capsule, length=7.2916, radius=6.585, shapeName=pelvis_capsule, center=<0 0.8436 0>}]
 thigh_l [{shapeType=Capsule, length=33.0941, radius=8.9059, shapeName=thigh_l_capsule, center=<-15.2592 -0.5472 2.4076>, rotation=<rot 90 0.7229 0>}]
 thigh_r [{
   shapeType=Capsule
   length=33
   radius=9.5
   shapeName=thigh_r_capsule
   center=<15.0489 2.0976 -1.3699>
   rotation=<rot 89.7176 -175.9117 -0.001>
  }]
}
constraints [
 {
  bone1Name=base_backJacket
  bone2Name=pelvis
  swing1LimitDegrees=45
  swing1Motion=Free
  swing2LimitDegrees=45
  swing2Motion=Free
  twistLimitDegrees=45
  twistMotion=Free
 }
 {bone1Name=dyn_backJacket_01, bone2Name=dyn_backJacket_00, swing1LimitDegrees=10, swing2LimitDegrees=30, twistLimitDegrees=2}
 {bone1Name=dyn_backJacket_02, bone2Name=dyn_backJacket_01, swing1LimitDegrees=10, swing2LimitDegrees=30, twistLimitDegrees=2}
 {bone1Name=dyn_backJacket_03, bone2Name=dyn_backJacket_02, swing1LimitDegrees=10, swing2LimitDegrees=30, twistLimitDegrees=2}
 {bone1Name=dyn_backJacket_04, bone2Name=dyn_backJacket_03, swing1LimitDegrees=10, swing2LimitDegrees=30, twistLimitDegrees=2}
 {bone1Name=base_backJacket_l, bone2Name=pelvis, swing1LimitDegrees=45, swing2LimitDegrees=45, twistLimitDegrees=45}
 {bone1Name=dyn_backJacket_l_01, bone2Name=dyn_backJacket_l_00, swing1LimitDegrees=10, swing2LimitDegrees=30, twistLimitDegrees=2}
 {bone1Name=dyn_backJacket_l_02, bone2Name=dyn_backJacket_l_01, swing1LimitDegrees=10, swing2LimitDegrees=30, twistLimitDegrees=2}
 {bone1Name=dyn_backJacket_l_03, bone2Name=dyn_backJacket_l_02, swing1LimitDegrees=10, swing2LimitDegrees=30, twistLimitDegrees=2}
 {bone1Name=dyn_backJacket_l_04, bone2Name=dyn_backJacket_l_03, swing1LimitDegrees=10, swing2LimitDegrees=30, twistLimitDegrees=2}
 {bone1Name=base_backJacket_r, bone2Name=pelvis, swing1LimitDegrees=45, swing2LimitDegrees=45, twistLimitDegrees=45}
 {bone1Name=dyn_backJacket_r_01, bone2Name=dyn_backJacket_r_00, swing1LimitDegrees=10, swing2LimitDegrees=30, twistLimitDegrees=2}
 {bone1Name=dyn_backJacket_r_02, bone2Name=dyn_backJacket_r_01, swing1LimitDegrees=10, swing2LimitDegrees=30, twistLimitDegrees=2}
 {bone1Name=dyn_backJacket_r_03, bone2Name=dyn_backJacket_r_02, swing1LimitDegrees=10, swing2LimitDegrees=30, twistLimitDegrees=2}
 {bone1Name=dyn_backJacket_r_04, bone2Name=dyn_backJacket_r_03, swing1LimitDegrees=10, swing2LimitDegrees=30, twistLimitDegrees=2}
 {bone1Name=base_frontJacket_l, bone2Name=pelvis, swing1LimitDegrees=45, swing2LimitDegrees=45, twistLimitDegrees=45}
 {bone1Name=dyn_frontJacket_l_01, bone2Name=dyn_frontJacket_l_00, swing1LimitDegrees=20, swing2LimitDegrees=30, twistLimitDegrees=15}
 {bone1Name=dyn_frontJacket_l_02, bone2Name=dyn_frontJacket_l_01, swing1LimitDegrees=10, swing2LimitDegrees=30, twistLimitDegrees=15}
 {bone1Name=dyn_frontJacket_l_03, bone2Name=dyn_frontJacket_l_02, swing1LimitDegrees=10, swing2LimitDegrees=30, twistLimitDegrees=15}
 {bone1Name=dyn_frontJacket_l_04, bone2Name=dyn_frontJacket_l_03, swing1LimitDegrees=10, swing2LimitDegrees=30, twistLimitDegrees=2}
 {bone1Name=base_frontJacket_r, bone2Name=pelvis, swing1LimitDegrees=45, swing2LimitDegrees=45, twistLimitDegrees=45}
 {bone1Name=dyn_frontJacket_r_01, bone2Name=dyn_frontJacket_r_00, swing1LimitDegrees=10, swing2LimitDegrees=30, twistLimitDegrees=15}
 {bone1Name=dyn_frontJacket_r_02, bone2Name=dyn_frontJacket_r_01, swing1LimitDegrees=10, swing2LimitDegrees=30, twistLimitDegrees=15}
 {bone1Name=dyn_frontJacket_r_03, bone2Name=dyn_frontJacket_r_02, swing1LimitDegrees=10, swing2LimitDegrees=30, twistLimitDegrees=2}
 {bone1Name=dyn_frontJacket_r_04, bone2Name=dyn_frontJacket_r_03, swing1LimitDegrees=10, swing2LimitDegrees=30, twistLimitDegrees=2}
 {bone1Name=base_midJacket_l, bone2Name=pelvis, swing1LimitDegrees=45, swing2LimitDegrees=45, twistLimitDegrees=45}
 {bone1Name=dyn_midJacket_l_01, bone2Name=dyn_midJacket_l_00, swing1LimitDegrees=10, swing2LimitDegrees=30, twistLimitDegrees=2}
 {bone1Name=dyn_midJacket_l_02, bone2Name=dyn_midJacket_l_01, swing1LimitDegrees=10, swing2LimitDegrees=30, twistLimitDegrees=2}
 {bone1Name=dyn_midJacket_l_03, bone2Name=dyn_midJacket_l_02, swing1LimitDegrees=10, swing2LimitDegrees=30, twistLimitDegrees=2}
 {bone1Name=dyn_midJacket_l_04, bone2Name=dyn_midJacket_l_03, swing1LimitDegrees=10, swing2LimitDegrees=30, twistLimitDegrees=2}
 {
  bone1Name=base_midJacket_r
  bone2Name=pelvis
  swing1LimitDegrees=45
  swing1Motion=Free
  swing2LimitDegrees=45
  swing2Motion=Free
  twistLimitDegrees=45
  twistMotion=Free
 }
 {bone1Name=dyn_midJacket_r_01, bone2Name=dyn_midJacket_r_00, swing1LimitDegrees=10, swing2LimitDegrees=30, twistLimitDegrees=2}
 {bone1Name=dyn_midJacket_r_02, bone2Name=dyn_midJacket_r_01, swing1LimitDegrees=10, swing2LimitDegrees=30, twistLimitDegrees=2}
 {bone1Name=dyn_midJacket_r_03, bone2Name=dyn_midJacket_r_02, swing1LimitDegrees=10, swing2LimitDegrees=30, twistLimitDegrees=2}
 {bone1Name=dyn_midJacket_r_04, bone2Name=dyn_midJacket_r_03, swing1LimitDegrees=10, swing2LimitDegrees=30, twistLimitDegrees=2}
 {
  bone1Name=dyn_phone
  bone2Name=pelvis
  swing1LimitDegrees=45
  swing1Motion=Free
  swing2LimitDegrees=45
  swing2Motion=Free
  twistLimitDegrees=45
  twistMotion=Free
 }
 {
  bone1Name=thigh_l
  bone2Name=pelvis
  swing1LimitDegrees=45
  swing1Motion=Locked
  swing2LimitDegrees=45
  swing2Motion=Locked
  twistLimitDegrees=45
  twistMotion=Locked
 }
 {
  bone1Name=thigh_r
  bone2Name=pelvis
  swing1LimitDegrees=45
  swing1Motion=Locked
  swing2LimitDegrees=45
  swing2Motion=Locked
  twistLimitDegrees=45
  twistMotion=Locked
 }
 {
  bone1Name=pelvis
  bone2Name=root
  swing1LimitDegrees=45
  swing1Motion=Free
  swing2LimitDegrees=45
  swing2Motion=Free
  twistLimitDegrees=45
  twistMotion=Free
 }
 {bone1Name=dyn_backJacket_00, bone2Name=base_backJacket, swing1LimitDegrees=0, swing2LimitDegrees=5, twistLimitDegrees=0}
 {bone1Name=dyn_backJacket_l_00, bone2Name=base_backJacket_l, swing1LimitDegrees=0, swing2LimitDegrees=5, twistLimitDegrees=0}
 {bone1Name=dyn_backJacket_r_00, bone2Name=base_backJacket_r, swing1LimitDegrees=0, swing2LimitDegrees=5, twistLimitDegrees=0}
 {bone1Name=dyn_frontJacket_l_00, bone2Name=base_frontJacket_l, swing1LimitDegrees=12, swing2LimitDegrees=15, twistLimitDegrees=0}
 {bone1Name=dyn_frontJacket_r_00, bone2Name=base_frontJacket_r, swing1LimitDegrees=12, swing2LimitDegrees=10, twistLimitDegrees=5}
 {bone1Name=dyn_midJacket_l_00, bone2Name=base_midJacket_l, swing1LimitDegrees=6, swing2LimitDegrees=5, twistLimitDegrees=0}
 {bone1Name=dyn_midJacket_r_00, bone2Name=base_midJacket_r, swing1LimitDegrees=6, swing2LimitDegrees=5, twistLimitDegrees=0}
 {
  bone1Name=spine_01
  bone2Name=pelvis
  swing1LimitDegrees=45
  swing1Motion=Free
  swing2LimitDegrees=45
  swing2Motion=Free
  twistLimitDegrees=45
  twistMotion=Free
 }
 {
  bone1Name=dyn_midJacket_r_04
  bone2Name=dyn_backJacket_r_03
  swing1LimitDegrees=45
  swing1Motion=Free
  swing2LimitDegrees=45
  swing2Motion=Free
  twistLimitDegrees=45
  twistMotion=Free
 }
 {
  bone1Name=dyn_frontJacket_r_04
  bone2Name=dyn_midJacket_r_04
  swing1LimitDegrees=45
  swing1Motion=Free
  swing2LimitDegrees=45
  swing2Motion=Free
  twistLimitDegrees=45
  twistMotion=Free
 }
 {
  bone1Name=dyn_frontJacket_r_02
  bone2Name=dyn_midJacket_r_02
  swing1LimitDegrees=45
  swing1Motion=Free
  swing2LimitDegrees=45
  swing2Motion=Free
  twistLimitDegrees=45
  twistMotion=Free
 }
 {
  bone1Name=dyn_midJacket_l_04
  bone2Name=dyn_backJacket_l_04
  swing1LimitDegrees=45
  swing1Motion=Free
  swing2LimitDegrees=45
  swing2Motion=Free
  twistLimitDegrees=45
 }
 {
  bone1Name=dyn_backJacket_l_04
  bone2Name=dyn_backJacket_04
  swing1LimitDegrees=45
  swing1Motion=Free
  swing2LimitDegrees=45
  swing2Motion=Free
  twistLimitDegrees=45
 }
 {
  bone1Name=dyn_backJacket_r_04
  bone2Name=dyn_backJacket_04
  swing1LimitDegrees=45
  swing1Motion=Free
  swing2LimitDegrees=45
  swing2Motion=Free
  twistLimitDegrees=45
 }
 {
  bone1Name=dyn_frontJacket_l_04
  bone2Name=dyn_midJacket_l_04
  swing1LimitDegrees=45
  swing1Motion=Free
  swing2LimitDegrees=45
  swing2Motion=Free
  twistLimitDegrees=45
 }
 {
  bone1Name=dyn_frontJacket_l_02
  bone2Name=dyn_midJacket_l_02
  swing1LimitDegrees=45
  swing1Motion=Free
  swing2LimitDegrees=45
  swing2Motion=Free
  twistLimitDegrees=45
  twistMotion=Free
 }
 {
  bone1Name=dyn_backJacket_r_02
  bone2Name=dyn_backJacket_02
  swing1LimitDegrees=45
  swing1Motion=Free
  swing2LimitDegrees=45
  swing2Motion=Free
  twistLimitDegrees=45
  twistMotion=Free
 }
 {
  bone1Name=dyn_backJacket_l_02
  bone2Name=dyn_backJacket_02
  swing1LimitDegrees=45
  swing1Motion=Free
  swing2LimitDegrees=45
  swing2Motion=Free
  twistLimitDegrees=45
  twistMotion=Free
 }
 {
  bone1Name=dyn_frontJacket_l_01
  bone2Name=dyn_midJacket_l_01
  swing1LimitDegrees=45
  swing1Motion=Free
  swing2LimitDegrees=45
  swing2Motion=Free
  twistLimitDegrees=45
  twistMotion=Free
 }
 {
  bone1Name=dyn_frontJacket_l_03
  bone2Name=dyn_midJacket_l_03
  swing1LimitDegrees=45
  swing1Motion=Free
  swing2LimitDegrees=45
  swing2Motion=Free
  twistLimitDegrees=45
  twistMotion=Free
 }
 {
  bone1Name=dyn_midJacket_l_02
  bone2Name=dyn_backJacket_l_01
  swing1LimitDegrees=45
  swing1Motion=Free
  swing2LimitDegrees=45
  swing2Motion=Free
  twistLimitDegrees=45
 }
 {
  bone1Name=dyn_frontJacket_r_03
  bone2Name=dyn_midJacket_r_03
  swing1LimitDegrees=45
  swing1Motion=Free
  swing2LimitDegrees=45
  swing2Motion=Free
  twistLimitDegrees=45
  twistMotion=Free
 }
]
