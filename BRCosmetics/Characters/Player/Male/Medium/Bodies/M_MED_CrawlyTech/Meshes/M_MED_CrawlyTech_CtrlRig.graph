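# Control Rig for the CrawlyTech medium male body cosmetic that drives deform bones and procedurally poses its coat, belt and accessories from body motion.
# After the shared DeformRig function, a sequence runs Cloth_Pose and Coat_Belt_Setup.
# Weighted parent constraints attach phone, buckle, belt and nametag bones, and left thigh swing-twist, remapped through curves, offsets some of them.
# Spherical pose readers on spine, thighs and calves blend constraints and transform offsets on jacket chains.
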ControlRigBlueprint /BRCosmetics/Characters/Player/Male/Medium/Bodies/M_MED_CrawlyTech/Meshes/M_MED_CrawlyTech_CtrlRig
graphs [
 ContainedGraph {nodes [
   n1 = RigVMFunctionEntryNode()
   SequenceAggregate = &RigVMFunction_Sequence(ExecuteContext=@n1.ExecuteContext)
   SequenceAggregate_1 = &RigVMFunction_Sequence(ExecuteContext=@SequenceAggregate.B)
   RigVMFunctionReturnNode(A=@SequenceAggregate.A, B=@SequenceAggregate_1.A, C=@SequenceAggregate_1.B)
  ]}
 New Function_ContainedGraph {nodes [
   RigVMFunctionEntryNode()
   > RigVMFunction_Sequence = RigVMAggregateNode()
   &RigUnit_ParentConstraint(
    AdvancedSettings=(RotationOrderForFilter=XZY)
    Child=(Type=Bone,Name="base_phone")
    Parents="((Item=(Type=Bone,Name=\"thigh_twist_01_l\"),Weight=0.0089),(Item=(Type=Bone,Name=\"deform_groin_l\"),Weight=0.0119),(Item=(Type=Bone,Name=\"spine_03\"),Weight=0.0526),(Item=(Type=Bone,Name=\"spine_01\"),Weight=0.0772),(Item=(Type=Bone,Name=\"spine_02\"),Weight=0.08 ...(truncated)"
    Weight=1
    ExecutePin=@RigVMFunction_Sequence.A
   )
   > &RigUnit_ParentConstraint(
    AdvancedSettings=(RotationOrderForFilter=XZY)
    Child=(Type=Bone,Name="base_beltBuckle")
    Parents=((Item=(Type=Bone,Name="spine_03"),Weight=0.0701),(Item=(Type=Bone,Name="spine_01"),Weight=0.0982),(Item=(Type=Bone,Name="spine_02"),Weight=0.1166),(Item=(Type=Bone,Name="pelvis"),Weight=0.7151))
    Weight=1
   )
   > &RigUnit_ParentConstraint(
    AdvancedSettings=(RotationOrderForFilter=XZY)
    Child=(Type=Bone,Name="base_belt_l")
    Parents=((Item=(Type=Bone,Name="deform_groin_l"),Weight=0.2265),(Item=(Type=Bone,Name="thigh_twist_01_l"),Weight=0.296),(Item=(Type=Bone,Name="pelvis"),Weight=0.4776))
    Weight=1
   )
   > &RigUnit_ParentConstraint(
    AdvancedSettings=(RotationOrderForFilter=XZY)
    Child=(Type=Bone,Name="base_belt")
    Parents="((Item=(Type=Bone,Name=\"deform_groin_r\"),Weight=0.0067),(Item=(Type=Bone,Name=\"spine_03\"),Weight=0.0118),(Item=(Type=Bone,Name=\"spine_01\"),Weight=0.0164),(Item=(Type=Bone,Name=\"spine_02\"),Weight=0.0239),(Item=(Type=Bone,Name=\"thigh_twist_01_l\"),Weight=0.08 ...(truncated)"
    Weight=1
   )
   > &RigUnit_ParentConstraint(
    AdvancedSettings=(RotationOrderForFilter=XZY)
    Child=(Type=Bone,Name="base_belt_r")
    Parents="((Item=(Type=Bone,Name=\"spine_03\"),Weight=0.0062),(Item=(Type=Bone,Name=\"spine_01\"),Weight=0.0088),(Item=(Type=Bone,Name=\"spine_02\"),Weight=0.0164),(Item=(Type=Bone,Name=\"thigh_twist_01_r\"),Weight=0.1345),(Item=(Type=Bone,Name=\"deform_groin_r\"),Weight=0.15 ...(truncated)"
    Weight=1
   )
   > &RigUnit_ParentConstraint(
    AdvancedSettings=(RotationOrderForFilter=XZY)
    Child=(Type=Bone,Name="base_nametag")
    Parents="((Item=(Type=Bone,Name=\"upperarm_twist_01_l\"),Weight=0.0372),(Item=(Type=Bone,Name=\"spine_02\"),Weight=0.0405),(Item=(Type=Bone,Name=\"spine_03\"),Weight=0.1241),(Item=(Type=Bone,Name=\"spine_05\"),Weight=0.1339),(Item=(Type=Bone,Name=\"spine_04\"),Weight=0.1829) ...(truncated)"
    Weight=1
   )
   RigVMFunctionReturnNode(ExecuteContext=@RigVMFunction_Sequence.F)
   RigVMFunction_MathQuaternionSwingTwist = &RigVMFunction_MathQuaternionSwingTwist(
    TwistAxis=(X=1,Y=0,Z=0)
    Input=&RigUnit_GetRelativeTransformForItem(Child=(Type=Bone,Name="thigh_r"), Parent=(Type=Null,Name="thigh_r_null"), bChildInitial=False, bParentInitial=False).RelativeTransform.Rotation
   )
   RigVMFunction_MathQuaternionToEuler = &RigVMFunction_MathQuaternionToEuler(RotationOrder=ZYX, Value=@RigVMFunction_MathQuaternionSwingTwist.Swing)
   RigVMVariableNode(Value=@RigVMFunction_MathQuaternionToEuler.Result, ExecuteContext=@RigVMFunction_Sequence.B) {variable=Leg_r_Swing_rot}
   n1 = &RigUnit_GetRelativeTransformForItem(Child=(Type=Bone,Name="thigh_l"), Parent=(Type=Null,Name="thigh_l_null"), bChildInitial=False, bParentInitial=False)
   > RigVMVariableNode(Value=&RigVMFunction_MathQuaternionToEuler(
     RotationOrder=ZYX
     Value=&RigVMFunction_MathQuaternionSwingTwist(TwistAxis=(X=1,Y=0,Z=0), Input=@n1.RelativeTransform.Rotation).Swing
    ).Result) {variable=Leg_l_Swing_rot}
   n2 = RigVMVariableNode() {variable=Leg_l_Swing_rot}
   n3 = &RigVMFunction_AnimEvalRichCurve(
    Curve="(EditorCurveData=(Keys=((InterpMode=RCIM_Cubic,TangentMode=RCTM_User,ArriveTangent=0.2796,LeaveTangent=0.2796),(InterpMode=RCIM_Cubic,TangentMode=RCTM_User,Time=0.3986,Value=0.2602,ArriveTangent=1.22,LeaveTangent=1.22),(InterpMode=RCIM_Cubic,TangentMode=RC ...(truncated)"
    SourceMaximum=-140
    SourceMinimum=0
    TargetMaximum=-44
    TargetMinimum=0
    Value=@n2.Value.Y
   )
   n4 = &RigVMFunction_AnimEvalRichCurve(
    Curve="(EditorCurveData=(Keys=((InterpMode=RCIM_Cubic,TangentMode=RCTM_User,ArriveTangent=0.2796,LeaveTangent=0.2796),(InterpMode=RCIM_Cubic,TangentMode=RCTM_User,Time=0.4047,Value=0.3757,ArriveTangent=1.2761,LeaveTangent=1.2761),(InterpMode=RCIM_Cubic,TangentMod ...(truncated)"
    SourceMaximum=-140
    SourceMinimum=0
    TargetMaximum=-44
    TargetMinimum=0
    Value=@n2.Value.Z
   )
   n5 = &RigVMFunction_AlphaInterp(
    InterpSpeedDecreasing=10
    InterpSpeedIncreasing=10
    Scale=1
    bClampResult=false
    bInterpResult=false
    bMapRange=true
    Value=&RigVMFunction_MathDoubleAdd(A=&RigVMFunction_MathDoubleMul(B=0.25, A=@n3.Result).Result, B=@n4.Result).Result
   )
   n6 = &RigVMFunction_MathDoubleMul(B=-0.5, A=@n5.Result)
   n7 = &RigVMFunction_MathDoubleMul(B=0.75, A=@n5.Result)
   &RigUnit_ModifyTransforms(
    ItemToModify="((Item=(Type=Bone,Name=\"dyn_phone\"),Transform=(Rotation=(X=0,Y=-0.1736,Z=0,W=0.9848),Translation=(X=0,Y=0,Z=0),Scale3D=(X=1,Y=1,Z=1))),(Item=(Type=Bone,Name=\"dyn_belt_l\"),Transform=(Rotation=(X=0,Y=0,Z=0,W=1),Translation=(X=0,Y=0,Z=0),Scale3D=(X=1,Y=1,Z=1) ...(truncated)"
    Weight=1
    WeightMinimum=0
    ItemToModify.0.Transform.Translation.Y=&RigVMFunction_MathDoubleMul(B=-0.01, A=@n5.Result).Result
    ItemToModify.0.Transform.Translation.X=&RigVMFunction_MathDoubleMul(B=0, A=@n5.Result).Result
    ItemToModify.1.Transform.Translation.Z=&RigVMFunction_MathDoubleMul(B=0.12, A=@n5.Result).Result
    ItemToModify.2.Transform.Translation.X=&RigVMFunction_MathDoubleMul(B=-0.07, A=@n5.Result).Result
    ItemToModify.2.Transform.Translation.Z=&RigVMFunction_MathDoubleMul(B=0.1, A=@n5.Result).Result
    ItemToModify.3.Transform.Translation.X=&RigVMFunction_MathDoubleMul(B=-0.1, A=@n5.Result).Result
    ItemToModify.3.Transform.Translation.Z=&RigVMFunction_MathDoubleMul(B=0.05, A=@n5.Result).Result
    ItemToModify.0.Transform.Translation.Z=&RigVMFunction_MathDoubleMul(B=-0.1, A=@n5.Result).Result
    ItemToModify.1.Transform.Translation.X=&RigVMFunction_MathDoubleMul(B=-0.2, A=@n5.Result).Result
    ItemToModify.0.Transform.Rotation=&RigVMFunction_MathQuaternionFromEuler(
     Euler=(X=20,Y=0,Z=0)
     Euler.X=&RigVMFunction_MathDoubleMul(B=-1.2, A=@n5.Result).Result
     Euler.Y=&RigVMFunction_MathDoubleMul(B=0, A=@n5.Result).Result
     Euler.Z=@n6.Result
    ).Result
    ItemToModify.2.Transform.Rotation=&RigVMFunction_MathQuaternionFromEuler(Euler=(X=0,Y=0,Z=0), Euler.X=&RigVMFunction_MathDoubleMul(B=1.125, A=@n5.Result).Result, Euler.Y=@n7.Result, Euler.Z=@n6.Result).Result
    ItemToModify.1.Transform.Rotation=&RigVMFunction_MathQuaternionFromEuler(
     Euler=(X=0,Y=0,Z=0)
     Euler.Y=&RigVMFunction_MathDoubleMul(B=2.5, A=@n5.Result).Result
     Euler.X=&RigVMFunction_MathDoubleMul(B=2, A=@n5.Result).Result
     Euler.Z=@n6.Result
    ).Result
    ItemToModify.3.Transform.Rotation=&RigVMFunction_MathQuaternionFromEuler(
     Euler=(X=0,Y=0,Z=0)
     Euler.Z=@n7.Result
     Euler.X=&RigVMFunction_MathDoubleMul(B=1.15, A=@n5.Result).Result
     Euler.Y=&RigVMFunction_MathDoubleMul(B=1.25, A=@n5.Result).Result
    ).Result
    ExecutePin=@RigVMFunction_Sequence.C
   )
  ]}
 New Function_ContainedGraph {nodes [
   RigVMFunctionEntryNode()
   > SequenceAggregate = RigVMAggregateNode()
   RigVMFunctionReturnNode(ExecuteContext=@SequenceAggregate.C)
   &RigUnit_ParentConstraint(
    AdvancedSettings=(InterpolationType=Average,RotationOrderForFilter=XZY)
    Child=(Type=Bone,Name="dyn_frontJacket_l_00")
    Filter=(TranslationFilter=(bX=true,bY=true,bZ=true),RotationFilter=(bX=true,bY=true,bZ=true),ScaleFilter=(bX=True,bY=True,bZ=True))
    Parents=((Item=(Type=Bone,Name="spine_01"),Weight=0.5),(Item=(Type=Bone,Name="pelvis"),Weight=0.5))
    Weight=1
    ExecutePin=@SequenceAggregate.A
   )
   > &RigUnit_ParentConstraint(
    AdvancedSettings=(InterpolationType=Average,RotationOrderForFilter=XZY)
    Child=(Type=Bone,Name="dyn_frontJacket_l_01")
    Filter=(TranslationFilter=(bX=true,bY=true,bZ=true),RotationFilter=(bX=true,bY=true,bZ=true),ScaleFilter=(bX=True,bY=True,bZ=True))
    Parents=((Item=(Type=Bone,Name="thigh_twist_01_l"),Weight=1),(Item=(Type=Bone,Name="pelvis"),Weight=0))
    Weight=1
   )
   > &RigUnit_ParentConstraint(
    AdvancedSettings=(InterpolationType=Average,RotationOrderForFilter=XZY)
    Child=(Type=Bone,Name="dyn_midJacket_l_00")
    Filter=(TranslationFilter=(bX=false,bY=false,bZ=false),RotationFilter=(bX=true,bY=true,bZ=true),ScaleFilter=(bX=True,bY=True,bZ=True))
    Parents=((Item=(Type=Bone,Name="thigh_twist_01_l"),Weight=0.5),(Item=(Type=Bone,Name="pelvis"),Weight=0.5))
    Weight=1
   )
   > n1 = &RigUnit_SphericalPoseReader(
    ActiveRegionScaleFactors=(PositiveWidth=2,NegativeWidth=2,PositiveHeight=2,NegativeHeight=2)
    ActiveRegionSize=0.1
    Debug=(bDrawDebug=false,DebugScale=25,DebugSegments=20,DebugThickness=0.25)
    DriverAxis=(X=0,Y=1,Z=0)
    DriverItem=(Type=Bone,Name="spine_04")
    FalloffRegionScaleFactors=(PositiveWidth=1.35,NegativeWidth=1.25,PositiveHeight=1.35,NegativeHeight=1.35)
    FalloffSize=0.5
    FlipHeightScaling=false
    FlipWidthScaling=true
    OptionalParentItem=(Type=Bone,Name="pelvis")
    RotationOffset=(X=-30,Y=180,Z=40)
   )
   > &RigUnit_ModifyTransforms(
    ItemToModify=((Item=(Type=Bone,Name="dyn_frontJacket_l_00"),Transform=(Rotation=(X=0,Y=0,Z=0,W=1),Translation=(X=0,Y=-5,Z=0),Scale3D=(X=1,Y=1,Z=1))))
    Weight=@n1.OutputParam
   )
   > n2 = &RigUnit_SphericalPoseReader(
    ActiveRegionScaleFactors=(PositiveWidth=0.1,NegativeWidth=1,PositiveHeight=0.1,NegativeHeight=0.1)
    ActiveRegionSize=0.1
    Debug=(bDrawDebug=false,bDraw2D=false,bDrawLocalAxes=false,DebugScale=25,DebugSegments=20,DebugThickness=0.25)
    DriverAxis=(X=-1,Y=0,Z=0)
    DriverItem=(Type=Bone,Name="thigh_twist_01_l")
    FalloffRegionScaleFactors=(PositiveWidth=1.35,NegativeWidth=1.7,PositiveHeight=1.35,NegativeHeight=1.35)
    FalloffSize=0.5
    OptionalParentItem=(Type=Bone,Name="spine_01")
    RotationOffset=(X=120,Y=0,Z=-30)
   )
   > &RigUnit_ParentConstraint(
    AdvancedSettings=(RotationOrderForFilter=XZY)
    Child=(Type=Bone,Name="dyn_frontJacket_l_01")
    Filter=(TranslationFilter=(bX=true,bY=true,bZ=true),RotationFilter=(bX=true,bY=true,bZ=true),ScaleFilter=(bX=True,bY=True,bZ=True))
    Parents=((Item=(Type=Bone,Name="thigh_twist_01_l"),Weight=0),(Item=(Type=Bone,Name="pelvis"),Weight=1))
    Weight=@n2.OutputParam
   )
   > &RigUnit_ParentConstraint(
    AdvancedSettings=(RotationOrderForFilter=XZY)
    Child=(Type=Bone,Name="dyn_midJacket_l_00")
    Parents=((Item=(Type=Bone,Name="thigh_twist_01_l"),Weight=0),(Item=(Type=Bone,Name="pelvis"),Weight=1))
    Weight=@n2.OutputParam
   )
   > &RigUnit_ModifyTransforms(
    ItemToModify="((Item=(Type=Bone,Name=\"dyn_frontJacket_l_01\"),Transform=(Rotation=(X=0,Y=0,Z=0,W=1),Translation=(X=0,Y=0,Z=0),Scale3D=(X=1,Y=1,Z=1))),(Item=(Type=Bone,Name=\"dyn_midJacket_l_00\"),Transform=(Rotation=(X=0,Y=0,Z=0,W=1),Translation=(X=0,Y=0,Z=2),Scale3D=(X=1, ...(truncated)"
    ItemToModify.0.Transform.Rotation=&RigVMFunction_MathQuaternionFromEuler(Euler=(X=0,Y=0,Z=0)).Result
    ItemToModify.1.Transform.Rotation=&RigVMFunction_MathQuaternionFromEuler(Euler=(X=-10,Y=0,Z=-5)).Result
    ItemToModify.2.Transform.Rotation=&RigVMFunction_MathQuaternionFromEuler(Euler=(X=0,Y=0,Z=0)).Result
    ItemToModify.3.Transform.Rotation=&RigVMFunction_MathQuaternionFromEuler(Euler=(X=0,Y=0,Z=0)).Result
    Weight=@n2.OutputParam
   )
   > n3 = &RigUnit_SphericalPoseReader(
    ActiveRegionScaleFactors=(PositiveWidth=2,NegativeWidth=0.2,PositiveHeight=1,NegativeHeight=2)
    ActiveRegionSize=0.1
    Debug=(bDrawDebug=false,DebugScale=25,DebugSegments=20,DebugThickness=0.25)
    DriverAxis=(X=-1,Y=0,Z=0)
    DriverItem=(Type=Bone,Name="calf_l")
    FalloffRegionScaleFactors=(PositiveWidth=1,NegativeWidth=0.9,PositiveHeight=3,NegativeHeight=1)
    FalloffSize=0.57
    FlipHeightScaling=false
    FlipWidthScaling=false
    OptionalParentItem=(Type=Bone,Name="thigh_l")
    RotationOffset=(X=-140,Y=-10,Z=30)
   )
   > &RigUnit_ParentConstraint(
    AdvancedSettings=(InterpolationType=Average,RotationOrderForFilter=XZY)
    Child=(Type=Bone,Name="dyn_backJacket_l_00")
    Filter=(TranslationFilter=(bX=false,bY=false,bZ=false),RotationFilter=(bX=false,bY=false,bZ=true),ScaleFilter=(bX=True,bY=True,bZ=True))
    Parents=((Item=(Type=Bone,Name="thigh_l"),Weight=0.7),(Item=(Type=Bone,Name="pelvis"),Weight=0.3))
    Weight=@n3.OutputParam
   )
   > &RigUnit_ParentConstraint(
    AdvancedSettings=(InterpolationType=Average,RotationOrderForFilter=XZY)
    Child=(Type=Bone,Name="dyn_backJacket_l_03")
    Filter=(TranslationFilter=(bX=false,bY=false,bZ=false),RotationFilter=(bX=false,bY=false,bZ=true),ScaleFilter=(bX=True,bY=True,bZ=True))
    Parents=((Item=(Type=Bone,Name="calf_l"),Weight=0.2),(Item=(Type=Bone,Name="dyn_Cloth_volume_l_C02"),Weight=0.8))
    Weight=@n3.OutputParam
   )
   n4 = &RigVMFunction_MathQuaternionFromEuler(Euler=(X=50,Y=0,Z=8))
   > &RigUnit_ModifyTransforms(
    ItemToModify="((Item=(Type=Bone,Name=\"dyn_backJacket_l_00\"),Transform=(Rotation=(X=0,Y=0,Z=0,W=1),Translation=(X=0,Y=0,Z=0),Scale3D=(X=1,Y=1,Z=1))),(Item=(Type=Bone,Name=\"dyn_backJacket_l_02\"),Transform=(Rotation=(X=0,Y=0,Z=0,W=1),Translation=(X=0,Y=0,Z=0),Scale3D=(X=1, ...(truncated)"
    ItemToModify.2.Transform.Rotation=&RigVMFunction_MathQuaternionFromEuler(Euler=(X=0,Y=0,Z=0)).Result
    ItemToModify.0.Transform.Rotation=&RigVMFunction_MathQuaternionFromEuler(Euler=(X=-10,Y=-10,Z=-30)).Result
    ItemToModify.1.Transform.Rotation=&RigVMFunction_MathQuaternionFromEuler(Euler=(X=0,Y=0,Z=0)).Result
    ItemToModify.3.Transform.Rotation=@n4.Result
    ItemToModify.4.Transform.Rotation=@n4.Result
    Weight=@n3.OutputParam
   )
   > n5 = &RigUnit_SphericalPoseReader(
    ActiveRegionScaleFactors=(PositiveWidth=2,NegativeWidth=3,PositiveHeight=1,NegativeHeight=2)
    ActiveRegionSize=0.1
    Debug=(bDrawDebug=false,bDraw2D=false,bDrawLocalAxes=false,DebugScale=25,DebugSegments=20,DebugThickness=0.25)
    DriverAxis=(X=-1,Y=0,Z=0)
    DriverItem=(Type=Bone,Name="thigh_twist_01_l")
    FalloffRegionScaleFactors=(PositiveWidth=1.35,NegativeWidth=1.7,PositiveHeight=1,NegativeHeight=0.9)
    FalloffSize=0.4
    OptionalParentItem=(Type=Bone,Name="spine_01")
    RotationOffset=(X=180,Y=0,Z=0)
   )
   > &RigUnit_ModifyTransforms(
    ItemToModify=((Item=(Type=Bone,Name="dyn_frontJacket_l_00"),Transform=(Rotation=(X=0.2586,Y=-0.0421,Z=0.0113,W=0.965),Translation=(X=0,Y=-5,Z=7),Scale3D=(X=1,Y=1,Z=1))))
    Weight=@n5.OutputParam
   )
   &RigUnit_ParentConstraint(
    AdvancedSettings=(InterpolationType=Average,RotationOrderForFilter=XZY)
    Child=(Type=Bone,Name="dyn_frontJacket_r_00")
    Filter=(TranslationFilter=(bX=true,bY=true,bZ=true),RotationFilter=(bX=true,bY=true,bZ=true),ScaleFilter=(bX=True,bY=True,bZ=True))
    Parents=((Item=(Type=Bone,Name="spine_01"),Weight=0.5),(Item=(Type=Bone,Name="pelvis"),Weight=0.5))
    Weight=1
    ExecutePin=@SequenceAggregate.B
   )
   > &RigUnit_ParentConstraint(
    AdvancedSettings=(InterpolationType=Average,RotationOrderForFilter=XZY)
    Child=(Type=Bone,Name="dyn_frontJacket_r_01")
    Filter=(TranslationFilter=(bX=true,bY=true,bZ=true),RotationFilter=(bX=true,bY=true,bZ=true),ScaleFilter=(bX=True,bY=True,bZ=True))
    Parents=((Item=(Type=Bone,Name="thigh_twist_01_r"),Weight=1),(Item=(Type=Bone,Name="pelvis"),Weight=0))
    Weight=1
   )
   > &RigUnit_ParentConstraint(
    AdvancedSettings=(InterpolationType=Average,RotationOrderForFilter=XZY)
    Child=(Type=Bone,Name="dyn_midJacket_r_00")
    Filter=(TranslationFilter=(bX=false,bY=false,bZ=false),RotationFilter=(bX=true,bY=true,bZ=true),ScaleFilter=(bX=True,bY=True,bZ=True))
    Parents=((Item=(Type=Bone,Name="thigh_twist_01_r"),Weight=0.5),(Item=(Type=Bone,Name="pelvis"),Weight=0.5))
    Weight=1
   )
   > n6 = &RigUnit_SphericalPoseReader(
    ActiveRegionScaleFactors=(PositiveWidth=2,NegativeWidth=2,PositiveHeight=2,NegativeHeight=2)
    ActiveRegionSize=0.1
    Debug=(bDrawDebug=false,DebugScale=25,DebugSegments=20,DebugThickness=0.25)
    DriverAxis=(X=0,Y=1,Z=0)
    DriverItem=(Type=Bone,Name="spine_04")
    FalloffRegionScaleFactors=(PositiveWidth=1.35,NegativeWidth=1.25,PositiveHeight=1.35,NegativeHeight=1.35)
    FalloffSize=0.6
    FlipHeightScaling=false
    FlipWidthScaling=true
    OptionalParentItem=(Type=Bone,Name="pelvis")
    RotationOffset=(X=-30,Y=0,Z=50)
   )
   > &RigUnit_ModifyTransforms(
    ItemToModify=((Item=(Type=Bone,Name="dyn_frontJacket_r_00"),Transform=(Rotation=(X=0,Y=0,Z=0,W=1),Translation=(X=0,Y=-5,Z=0),Scale3D=(X=1,Y=1,Z=1))))
    Weight=@n6.OutputParam
   )
   > n7 = &RigUnit_SphericalPoseReader(
    ActiveRegionScaleFactors=(PositiveWidth=2,NegativeWidth=1,PositiveHeight=2,NegativeHeight=2)
    ActiveRegionSize=0.1
    Debug=(bDrawDebug=false,DebugScale=25,DebugSegments=20,DebugThickness=0.25)
    DriverAxis=(X=1,Y=0,Z=0)
    DriverItem=(Type=Bone,Name="thigh_twist_01_r")
    FalloffRegionScaleFactors=(PositiveWidth=1.35,NegativeWidth=1.3,PositiveHeight=1.3,NegativeHeight=1.35)
    FalloffSize=0.5
    FlipHeightScaling=false
    FlipWidthScaling=true
    OptionalParentItem=(Type=Bone,Name="spine_01")
    RotationOffset=(X=-60,Y=0,Z=-30)
   )
   > &RigUnit_ParentConstraint(
    AdvancedSettings=(RotationOrderForFilter=XZY)
    Child=(Type=Bone,Name="dyn_frontJacket_r_01")
    Filter=(TranslationFilter=(bX=true,bY=true,bZ=true),RotationFilter=(bX=True,bY=True,bZ=True),ScaleFilter=(bX=True,bY=True,bZ=True))
    Parents=((Item=(Type=Bone,Name="thigh_twist_01_r"),Weight=0),(Item=(Type=Bone,Name="pelvis"),Weight=1))
    Weight=@n7.OutputParam
   )
   > &RigUnit_ParentConstraint(
    AdvancedSettings=(RotationOrderForFilter=XZY)
    Child=(Type=Bone,Name="dyn_midJacket_r_00")
    Parents=((Item=(Type=Bone,Name="thigh_twist_01_r"),Weight=0),(Item=(Type=Bone,Name="pelvis"),Weight=1))
    Weight=@n7.OutputParam
   )
   > &RigUnit_ModifyTransforms(
    ItemToModify="((Item=(Type=Bone,Name=\"dyn_frontJacket_r_01\"),Transform=(Rotation=(X=0,Y=0,Z=0,W=1),Translation=(X=0,Y=-2,Z=0),Scale3D=(X=1,Y=1,Z=1))),(Item=(Type=Bone,Name=\"dyn_midJacket_r_00\"),Transform=(Rotation=(X=0,Y=0,Z=0,W=1),Translation=(X=0,Y=0,Z=0),Scale3D=(X=1 ...(truncated)"
    ItemToModify.1.Transform.Rotation=&RigVMFunction_MathQuaternionFromEuler(Euler=(X=10,Y=0,Z=5)).Result
    ItemToModify.3.Transform.Rotation=&RigVMFunction_MathQuaternionFromEuler(Euler=(X=0,Y=0,Z=0)).Result
    ItemToModify.2.Transform.Rotation=&RigVMFunction_MathQuaternionFromEuler(Euler=(X=0,Y=0,Z=0)).Result
    ItemToModify.0.Transform.Rotation=&RigVMFunction_MathQuaternionFromEuler(Euler=(X=-20,Y=0,Z=0)).Result
    Weight=@n7.OutputParam
   )
   > n8 = &RigUnit_SphericalPoseReader(
    ActiveRegionScaleFactors=(PositiveWidth=2,NegativeWidth=0.2,PositiveHeight=1,NegativeHeight=2)
    ActiveRegionSize=0.1
    Debug=(bDrawDebug=false,DebugScale=25,DebugSegments=20,DebugThickness=0.25)
    DriverAxis=(X=1,Y=0,Z=0)
    DriverItem=(Type=Bone,Name="calf_r")
    FalloffRegionScaleFactors=(PositiveWidth=1,NegativeWidth=0.9,PositiveHeight=3,NegativeHeight=1)
    FalloffSize=0.57
    FlipHeightScaling=false
    FlipWidthScaling=true
    OptionalParentItem=(Type=Bone,Name="thigh_r")
    RotationOffset=(X=40,Y=-10,Z=30)
   )
   > &RigUnit_ParentConstraint(
    AdvancedSettings=(InterpolationType=Average,RotationOrderForFilter=XZY)
    Child=(Type=Bone,Name="dyn_backJacket_r_00")
    Filter=(TranslationFilter=(bX=false,bY=false,bZ=false),RotationFilter=(bX=false,bY=false,bZ=true),ScaleFilter=(bX=True,bY=True,bZ=True))
    Parents=((Item=(Type=Bone,Name="thigh_r"),Weight=0.7),(Item=(Type=Bone,Name="pelvis"),Weight=0.3))
    Weight=@n8.OutputParam
   )
   > &RigUnit_ParentConstraint(
    AdvancedSettings=(InterpolationType=Average,RotationOrderForFilter=XZY)
    Child=(Type=Bone,Name="dyn_backJacket_r_03")
    Filter=(TranslationFilter=(bX=false,bY=false,bZ=false),RotationFilter=(bX=false,bY=false,bZ=true),ScaleFilter=(bX=True,bY=True,bZ=True))
    Parents=((Item=(Type=Bone,Name="calf_r"),Weight=0.2),(Item=(Type=Bone,Name="dyn_Cloth_volume_r_C02"),Weight=0.8))
    Weight=@n8.OutputParam
   )
   n9 = &RigVMFunction_MathQuaternionFromEuler(Euler=(X=-50,Y=-15,Z=-8))
   > &RigUnit_ModifyTransforms(
    ItemToModify="((Item=(Type=Bone,Name=\"dyn_backJacket_r_00\"),Transform=(Rotation=(X=0,Y=0,Z=0,W=1),Translation=(X=0,Y=0,Z=0),Scale3D=(X=1,Y=1,Z=1))),(Item=(Type=Bone,Name=\"dyn_backJacket_r_02\"),Transform=(Rotation=(X=0,Y=0,Z=0,W=1),Translation=(X=0,Y=0,Z=0),Scale3D=(X=1, ...(truncated)"
    ItemToModify.0.Transform.Rotation=&RigVMFunction_MathQuaternionFromEuler(Euler=(X=10,Y=-10,Z=30)).Result
    ItemToModify.2.Transform.Rotation=&RigVMFunction_MathQuaternionFromEuler(Euler=(X=0,Y=0,Z=0)).Result
    ItemToModify.1.Transform.Rotation=&RigVMFunction_MathQuaternionFromEuler(Euler=(X=0,Y=0,Z=0)).Result
    ItemToModify.3.Transform.Rotation=@n9.Result
    ItemToModify.4.Transform.Rotation=@n9.Result
    Weight=@n8.OutputParam
   )
   RigVMRerouteNode(Value=())
  ]}
 RigVMFunctionLibrary {nodes [
   Cloth_Pose = RigVMCollapseNode(ExecuteContext=())
   Coat_Belt_Setup = RigVMCollapseNode(ExecuteContext=())
  ]}
 RigVMFunction_Sequence_1_ContainedGraph {nodes [
   n1 = RigVMFunctionEntryNode()
   RigVMFunction_Sequence = &RigVMFunction_Sequence(ExecuteContext=@n1.ExecuteContext)
   n2 = &RigVMFunction_Sequence(ExecuteContext=@RigVMFunction_Sequence.B)
   n3 = &RigVMFunction_Sequence(ExecuteContext=@n2.B)
   n4 = &RigVMFunction_Sequence(ExecuteContext=@n3.B)
   n5 = &RigVMFunction_Sequence(ExecuteContext=@n4.B)
   RigVMFunctionReturnNode(A=@RigVMFunction_Sequence.A, B=@n2.A, C=@n3.A, D=@n4.A, E=@n5.A, F=@n5.B)
  ]}
 RigVMModel {nodes [
   RigUnit_BeginExecution = &RigUnit_BeginExecution()
   FN_DeformRig_FNC = RigVMFunctionReferenceNode(
    DeformRig_ Curve="(deformRig_l_upperarm_up_alpha_72_#1=False,deformRig_r_upperarm_up_alpha_73_#2=False,deformRig_l_upperarm_fr_alpha_74_#3=False,deformRig_r_upperarm_fr_alpha_75_CB51B8 ...(truncated)"
    DeformRig_Bool="(upperarm_1_#4=True,elbow_3_#5=True,wrist_7_#6=True,groin_8_#7=True,glute_17_#8=True,knee_9_40C408324CA261 ...(truncated)"
    DeformRig_Debug=(upperarm_debug_18_#4=False,wrist_debug_20_#6=False,lat_debug_26_#9=False,groinGlute_debug_27_#7=False)
    DeformRig_Value="(upperarm_up_18_#10=(Rotation=(X=0.0,Y=-0.0,Z=0.0,W=1.0),Translation=(X=0.0,Y=0.0,Z=8.0),Scale3D=(X=0.70,Y=1.0,Z=1.0)),upperarm_fwd_fr_19_#11=(Rotation=(X=0.0,Y=-0.0,Z=0.0,W=1.0),Translation=(X=0.0, ...(truncated)"
    ExecuteContext=@RigUnit_BeginExecution.ExecutePin
   )
   RigVMFunction_Sequence = &RigVMFunction_Sequence(ExecuteContext=@FN_DeformRig_FNC.ExecuteContext)
   Cloth_Pose = RigVMFunctionReferenceNode(ExecuteContext=@RigVMFunction_Sequence.A)
   Coat_Belt_Setup_1 = RigVMFunctionReferenceNode(ExecuteContext=@RigVMFunction_Sequence.B)
   setFKCtrl = RigVMFunctionReferenceNode(ExecuteContext=(), setFK_Ctrl=false)
  ]}
 RigVMModel Rig {nodes [
   n1 = &RigUnit_PrepareForExecution()
   addNullObject = RigVMFunctionReferenceNode(
    Custom_Joints=()
    Options_Null="(pelvis_1_#12=False,spine_3_#13=False,clavicle_5_#14=False,arm_7_#15=False,leg_9_#16=true,head_11_46DC15EF465CD ...(truncated)"
    ExecuteContext=@n1.ExecutePin
   )
   addFKCtrl = RigVMFunctionReferenceNode(
    Custom_FKCtrl=()
    ExecuteContext=()
    Options_FKCtrl="(pelvis_1_#12=true,spine_3_#13=true,clavicle_5_#14=False,arm_7_#15=False,leg_9_#16=true,head_11_46DC15EF465CD29 ...(truncated)"
    Shape_Settings="(pelvis_2_#17=(bVisible=True,Name=\"Hexagon_Thin\",Color=(R=0.0117,G=0.8281,B=1,A=1),Transform=(Rotation=(X=0,Y=-0.7071,Z=0,W=0.7071),Translation=(X=2,Y=0,Z=0),Scale3D=(X=5,Y=5,Z=7))),spine_4_#18=(bVi ...(truncated)"
   )
  ]}
]
hierarchy [
 root {loc=<0 0 0>, rot=<0 0 0 1>}
 pelvis {parent=root, loc=<0 -0.7071 93.1191>, rot=<0.0127 0.707 -0.0127 -0.707>}
 spine_01 {parent=pelvis, loc=<6.7494 -0.6388 0>, rot=<0 0 0.034 -0.9994>}
 spine_02 {parent=spine_01, loc=<7.2518 -0.035 0>, rot=<0 0 0.0238 -0.9997>}
 spine_03 {parent=spine_02, loc=<6.0104 0 0>, rot=<0 0 0.0149 -0.9999>}
 spine_04 {parent=spine_03, loc=<8.1024 0 0>, rot=<0 0 -0.0132 -0.9999>}
 spine_05 {parent=spine_04, loc=<10.0146 0 0>, rot=<0 0 -0.0015 -1>}
 neck_01 {parent=spine_05, loc=<11.3138 0.1529 0>, rot=<0 0 -0.2003 -0.9797>}
 neck_02 {parent=neck_01, loc=<4.5627 0 0>, rot=<0 0 0.0504 -0.9987>}
 head {parent=neck_02, loc=<5.3914 0 0>, rot=<0 0 0.115 -0.9934>}
 attach_cape {parent=spine_05, loc=<9.017 -11.714 0>, rot=<0.0616 -0.7044 -0.0616 -0.7044>}
 cape {parent=attach_cape, type=Null, loc=<0 3.7399 0>, rot=<0 0 0 1>}
 attach_backpack {parent=spine_05, loc=<-0.1229 -13.473 0>, rot=<-0.0093 -0.707 0.0093 -0.707>}
 Backpack_BR {parent=attach_backpack, type=Null, loc=<0 4.0351 -0.2223>, rot=<-0.0453 0 0 0.999>}
 clavicle_l {parent=spine_05, loc=<4.3575 -0.1466 -2.3223>, rot=<0.0465 -0.7289 -0.009 -0.6829>}
 upperarm_l {parent=clavicle_l, loc=<15.5111 0 0>, rot=<0.0226 -0.3234 -0.0088 -0.9459>}
 lowerarm_l {parent=upperarm_l, loc=<26.6514 0 0>, rot=<0.0391 -0.0099 -0.2453 -0.9686>}
 hand_l {parent=lowerarm_l, loc=<22.3491 0 0>, rot=<-0.7081 -0.0507 0.0975 -0.6975>}
 thumb_01_l {parent=hand_l, loc=<1.3118 -0.3191 -2.459>, rot=<0.6546 -0.4042 0.0664 -0.6354>}
 thumb_02_l {parent=thumb_01_l, loc=<5.9336 -0.0448 0.0401>, rot=<-0.0037 0.0213 0.298 -0.9543>}
 thumb_03_l {parent=thumb_02_l, loc=<2.2223 0.3682 -0.0907>, rot=<0.0118 -0.013 0.0066 -0.9998>}
 index_metacarpal_l {parent=hand_l, loc=<1.6535 0.0183 -2.96>, rot=<0.0158 -0.0124 -0.0173 -0.9996>}
 index_01_l {parent=index_metacarpal_l, loc=<8.3008 0.0357 -0.0524>, rot=<-0.0044 0.0485 0.0924 -0.9945>}
 index_02_l {parent=index_01_l, loc=<3.3587 -0.3004 0.0452>, rot=<-0.002 0.0027 0.3225 -0.9466>}
 index_03_l {parent=index_02_l, loc=<2.5578 -0.123 -0.0252>, rot=<-0.002 0.0067 0.2558 -0.9667>}
 middle_metacarpal_l {parent=hand_l, loc=<2.1856 -0.1725 -0.5984>, rot=<0.006 0.0168 -0.0329 -0.9993>}
 middle_01_l {parent=middle_metacarpal_l, loc=<7.3275 -0.2447 0.0129>, rot=<-0.0208 0.0042 0.0858 -0.9961>}
 middle_02_l {parent=middle_01_l, loc=<4.0123 0.1401 0.1175>, rot=<-0.0218 0.0258 0.3512 -0.9357>}
 middle_03_l {parent=middle_02_l, loc=<2.6077 0.0534 0.1373>, rot=<0.0082 0.0188 0.1708 -0.9851>}
 ring_metacarpal_l {parent=hand_l, loc=<2.2074 -0.3625 1.3638>, rot=<-0.0284 0.0722 -0.0458 -0.9959>}
 ring_01_l {parent=ring_metacarpal_l, loc=<7.0295 -0.6197 -0.1446>, rot=<-0.0489 -0.0252 0.1892 -0.9804>}
 ring_02_l {parent=ring_01_l, loc=<3.6174 0.4647 -0.1027>, rot=<0 0 0.2806 -0.9598>}
 ring_03_l {parent=ring_02_l, loc=<2.4152 0.114 0.0199>, rot=<0 0 0.1189 -0.9929>}
 pinky_metacarpal_l {parent=hand_l, loc=<1.7616 -0.9753 3.3814>, rot=<-0.0397 0.0814 -0.069 -0.9935>}
 pinky_01_l {parent=pinky_metacarpal_l, loc=<7.1709 -0.4831 -0.1094>, rot=<-0.0654 -0.0035 0.2242 -0.9723>}
 pinky_02_l {parent=pinky_01_l, loc=<2.8274 0.0893 -0.1043>, rot=<0.0037 0.0011 0.2854 -0.9584>}
 pinky_03_l {parent=pinky_02_l, loc=<2.2046 0.0307 -0.0078>, rot=<-0.0002 0.0098 0.1445 -0.9895>}
 deform_wrist_dn_base_l {parent=hand_l, loc=<0 -2.2342 0>, rot=<0 0 0 -1>}
 deform_wrist_dn_l {parent=deform_wrist_dn_base_l, loc=<0 0 0>, rot=<0 0 0 1>}
 deform_wrist_up_base_l {parent=hand_l, loc=<0 2.2682 0>, rot=<0 0 0 -1>}
 deform_wrist_up_l {parent=deform_wrist_up_base_l, loc=<0 0 0>, rot=<0 0 0 1>}
 lowerarm_twist_02_l {parent=lowerarm_l, loc=<9.6199 0 0>, rot=<0 0 0 -1>}
 lowerarm_twist_01_l {parent=lowerarm_l, loc=<17.8549 0 0>, rot=<0 0 0 -1>}
 deform_elbow_in_base_l {parent=lowerarm_l, loc=<0.2112 5.259 0.8274>, rot=<0 0 0 -1>}
 deform_elbow_in_l {parent=deform_elbow_in_base_l, loc=<0 0 0>, rot=<0 0 0 1>}
 deform_elbow_out_base_l {parent=lowerarm_l, loc=<0 -4 0>, rot=<0 0 0 -1>}
 deform_elbow_out_l {parent=deform_elbow_out_base_l, loc=<0 0 0>, rot=<0 0 0 1>}
 upperarm_twist_01_l {parent=upperarm_l, loc=<0.0041 0 0>, rot=<0 0 0 -1>}
 deform_upperarm_bk_base_l {parent=upperarm_twist_01_l, loc=<1.0813 -7.2193 -1.9453>, rot=<0 0 0 1>}
 deform_upperarm_bk_l {parent=deform_upperarm_bk_base_l, loc=<0 0 0>, rot=<0 0 0 1>}
 deform_upperarm_fr_base_l {parent=upperarm_twist_01_l, loc=<-0.3691 7.0506 -2.765>, rot=<0 0 0 1>}
 deform_upperarm_fr_l {parent=deform_upperarm_fr_base_l, loc=<0 0 0>, rot=<0 0 0 1>}
 deform_upperarm_up_base_l {parent=upperarm_twist_01_l, loc=<-3.2216 -1.1391 3.3311>, rot=<0 0 0 1>}
 deform_upperarm_up_l {parent=deform_upperarm_up_base_l, loc=<0 0 0>, rot=<0 0 0 1>}
 upperarm_twist_02_l {parent=upperarm_l, loc=<13.7941 0 0>, rot=<0 0 0 -1>}
 clavicle_r {parent=spine_05, loc=<4.3575 -0.1466 2.3223>, rot=<0.7289 0.0465 -0.6829 0.009>}
 upperarm_r {parent=clavicle_r, loc=<-15.5111 0 0>, rot=<0.0226 -0.3234 -0.0088 -0.9459>}
 lowerarm_r {parent=upperarm_r, loc=<-26.6514 0 0>, rot=<0.0391 -0.0099 -0.2453 -0.9686>}
 hand_r {parent=lowerarm_r, loc=<-22.3491 0 0>, rot=<-0.7081 -0.0507 0.0975 -0.6975>}
 thumb_01_r {parent=hand_r, loc=<-1.3118 0.3191 2.459>, rot=<0.6546 -0.4042 0.0664 -0.6354>}
 thumb_02_r {parent=thumb_01_r, loc=<-5.9336 0.0448 -0.0401>, rot=<-0.0037 0.0213 0.298 -0.9543>}
 thumb_03_r {parent=thumb_02_r, loc=<-2.2223 -0.3682 0.0907>, rot=<0.0118 -0.013 0.0066 -0.9998>}
 index_metacarpal_r {parent=hand_r, loc=<-1.6535 -0.0183 2.96>, rot=<0.0158 -0.0124 -0.0173 -0.9996>}
 index_01_r {parent=index_metacarpal_r, loc=<-8.3008 -0.0357 0.0524>, rot=<-0.0044 0.0485 0.0924 -0.9945>}
 index_02_r {parent=index_01_r, loc=<-3.3587 0.3004 -0.0452>, rot=<-0.002 0.0027 0.3225 -0.9466>}
 index_03_r {parent=index_02_r, loc=<-2.5578 0.123 0.0252>, rot=<-0.002 0.0067 0.2558 -0.9667>}
 middle_metacarpal_r {parent=hand_r, loc=<-2.1856 0.1725 0.5984>, rot=<0.006 0.0168 -0.0329 -0.9993>}
 middle_01_r {parent=middle_metacarpal_r, loc=<-7.3275 0.2447 -0.0129>, rot=<-0.0208 0.0042 0.0858 -0.9961>}
 middle_02_r {parent=middle_01_r, loc=<-4.0123 -0.1401 -0.1175>, rot=<-0.0218 0.0258 0.3512 -0.9357>}
 middle_03_r {parent=middle_02_r, loc=<-2.6077 -0.0534 -0.1373>, rot=<0.0082 0.0188 0.1708 -0.9851>}
 ring_metacarpal_r {parent=hand_r, loc=<-2.2074 0.3625 -1.3638>, rot=<-0.0284 0.0722 -0.0458 -0.9959>}
 ring_01_r {parent=ring_metacarpal_r, loc=<-7.0295 0.6197 0.1446>, rot=<-0.0489 -0.0252 0.1892 -0.9804>}
 ring_02_r {parent=ring_01_r, loc=<-3.6174 -0.4647 0.1027>, rot=<0 0 0.2806 -0.9598>}
 ring_03_r {parent=ring_02_r, loc=<-2.4152 -0.114 -0.0199>, rot=<0 0 0.1189 -0.9929>}
 pinky_metacarpal_r {parent=hand_r, loc=<-1.7616 0.9753 -3.3814>, rot=<-0.0397 0.0814 -0.069 -0.9935>}
 pinky_01_r {parent=pinky_metacarpal_r, loc=<-7.1709 0.4831 0.1094>, rot=<-0.0654 -0.0035 0.2242 -0.9723>}
 pinky_02_r {parent=pinky_01_r, loc=<-2.8274 -0.0893 0.1043>, rot=<0.0037 0.0011 0.2854 -0.9584>}
 pinky_03_r {parent=pinky_02_r, loc=<-2.2046 -0.0307 0.0078>, rot=<-0.0002 0.0098 0.1445 -0.9895>}
 deform_wrist_dn_base_r {parent=hand_r, loc=<-0.0004 2.2345 -0.0001>, rot=<0 0 0 -1>}
 deform_wrist_dn_r {parent=deform_wrist_dn_base_r, loc=<0 0 0>, rot=<0 0 0 1>}
 deform_wrist_up_base_r {parent=hand_r, loc=<0.0003 -2.2684 0.0001>, rot=<0 0 0 -1>}
 deform_wrist_up_r {parent=deform_wrist_up_base_r, loc=<0 0 0>, rot=<0 0 0 1>}
 lowerarm_twist_02_r {parent=lowerarm_r, loc=<-9.6199 0 0>, rot=<0 0 0 -1>}
 lowerarm_twist_01_r {parent=lowerarm_r, loc=<-17.8549 0 0>, rot=<0 0 0 -1>}
 deform_elbow_in_base_r {parent=lowerarm_r, loc=<-0.2111 -5.2591 -0.8275>, rot=<0 0 0 -1>}
 deform_elbow_in_r {parent=deform_elbow_in_base_r, loc=<-0.0001 0 0.0002>, rot=<0 0 0 1>}
 deform_elbow_out_base_r {parent=lowerarm_r, loc=<0 4 -0.0001>, rot=<0 0 0 -1>}
 deform_elbow_out_r {parent=deform_elbow_out_base_r, loc=<0 0 0>, rot=<0 0 0 1>}
 upperarm_twist_01_r {parent=upperarm_r, loc=<-0.0041 0 0>, rot=<0 0 0 -1>}
 deform_upperarm_bk_base_r {parent=upperarm_twist_01_r, loc=<-1.0814 7.2192 1.9455>, rot=<0 0 0 1>}
 deform_upperarm_bk_r {parent=deform_upperarm_bk_base_r, loc=<0 0 0>, rot=<0 0 0 1>}
 deform_upperarm_fr_base_r {parent=upperarm_twist_01_r, loc=<0.3689 -7.0506 2.7652>, rot=<0 0 0 1>}
 deform_upperarm_fr_r {parent=deform_upperarm_fr_base_r, loc=<0 0 0>, rot=<0 0 0 1>}
 deform_upperarm_up_base_r {parent=upperarm_twist_01_r, loc=<3.2214 1.1391 -3.3308>, rot=<0 0 0 1>}
 deform_upperarm_up_r {parent=deform_upperarm_up_base_r, loc=<0 0 0>, rot=<0 0 0 -1>}
 upperarm_twist_02_r {parent=upperarm_r, loc=<-13.7941 0 0>, rot=<0 0 0 -1>}
 base_tie {parent=spine_05, loc=<8.6107 6.6526 -0.0101>, rot=<-0.2063 -0.6784 -0.6763 -0.1994>}
 dyn_tie_01 {parent=base_tie, loc=<0 0 0>, rot=<0 0 0 1>}
 dyn_tie_02 {parent=dyn_tie_01, loc=<6.2365 0.0004 -0.6587>, rot=<0 -0.1609 0 -0.987>}
 dyn_tie_03 {parent=dyn_tie_02, loc=<7.2897 0.0002 -0.2071>, rot=<0 -0.1106 0 -0.9939>}
 dyn_tie_04 {parent=dyn_tie_03, loc=<7.6816 -0.0001 0.297>, rot=<0 -0.013 0 -0.9999>}
 dyn_tie_05 {parent=dyn_tie_04, loc=<7.2785 0 0.0767>, rot=<0 -0.0261 0 -0.9997>}
 dyn_tie_06 {parent=dyn_tie_05, loc=<6.8373 0.0001 -0.1682>, rot=<0 -0.0147 0 -0.9999>}
 deform_pec_aim_base_l {parent=spine_04, loc=<7.8732 12.6388 -2.4123>, rot=<0.0102 -0.4602 0.3979 -0.7936>}
 deform_pec_aim_l {parent=deform_pec_aim_base_l, loc=<0 0 0>, rot=<0 0 0 1>}
 deform_pec_base_l {parent=deform_pec_aim_l, loc=<10.766 0 0>, rot=<0 0 0 1>}
 deform_pec_l {parent=deform_pec_base_l, loc=<0.055 0 0>, rot=<0 0 0 1>}
 deform_pec_aim_base_r {parent=spine_04, loc=<7.8727 12.6388 2.4123>, rot=<-0.4602 -0.0102 0.7936 0.3979>}
 deform_pec_aim_r {parent=deform_pec_aim_base_r, loc=<0 0 0>, rot=<0 0 0 1>}
 deform_pec_base_r {parent=deform_pec_aim_r, loc=<-10.7662 -0.0006 -0.0007>, rot=<0 0 0 1>}
 deform_pec_r {parent=deform_pec_base_r, loc=<0 0 0>, rot=<0 0 0 1>}
 base_nametag {parent=spine_04, loc=<9.3839 7.0623 -10.5832>, rot=<-0.3239 -0.5794 -0.1837 -0.725>}
 dyn_nametag {parent=base_nametag, loc=<0 0 0>, rot=<0 0 0 1>}
 thigh_l {parent=pelvis, loc=<-2.5693 -0.0109 -7.7501>, rot=<0.0012 0.0572 0.0155 -0.9982>}
 calf_l {parent=thigh_l, loc=<-39.0484 0 0>, rot=<0.0006 -0.0161 -0.0391 -0.9991>}
 foot_l {parent=calf_l, loc=<-42.561 0.1077 0.4292>, rot=<0.0418 -0.0455 0.0343 -0.9975>}
 ball_l {parent=foot_l, loc=<-6.1599 13.8264 0.6806>, rot=<0.0001 -0.0001 -0.7063 -0.7079>}
 calf_twist_02_l {parent=calf_l, loc=<-18.8064 -0.0967 0.0397>, rot=<0.0237 0.0095 -0.0035 -0.9997>}
 calf_twist_01_l {parent=calf_l, loc=<-35.813 -0.0967 0.0397>, rot=<0.0237 0.0095 -0.0035 -0.9997>}
 deform_knee_in_base_l {parent=calf_l, loc=<1.298 -5.2957 -0.0496>, rot=<0 0 0 -1>}
 deform_knee_in_l {parent=deform_knee_in_base_l, loc=<0 0 0>, rot=<0 0 0 1>}
 deform_knee_out_base_l {parent=calf_l, loc=<0 4 0>, rot=<0 0 0 -1>}
 deform_knee_out_l {parent=deform_knee_out_base_l, loc=<0 0 0>, rot=<0 0 0 1>}
 thigh_twist_01_l {parent=thigh_l, loc=<-21.6303 0 0>, rot=<0.0013 0 0 -1>}
 deform_groin_base_l {parent=thigh_l, loc=<-2.1908 8.1142 -1.7442>, rot=<0 0 0 -1>}
 deform_groin_l {parent=deform_groin_base_l, loc=<0 0 0>, rot=<0 0 0 1>}
 thigh_r {parent=pelvis, loc=<-2.5693 -0.0109 7.7501>, rot=<0.0572 -0.0012 0.9982 0.0155>}
 calf_r {parent=thigh_r, loc=<39.0484 0 0>, rot=<0.0006 -0.0161 -0.0391 -0.9991>}
 foot_r {parent=calf_r, loc=<42.561 -0.1077 -0.4292>, rot=<0.0418 -0.0455 0.0343 -0.9975>}
 ball_r {parent=foot_r, loc=<6.1599 -13.8264 -0.6806>, rot=<0.0001 -0.0001 -0.7063 -0.7079>}
 calf_twist_02_r {parent=calf_r, loc=<18.8064 0.0967 -0.0397>, rot=<0.0237 0.0095 -0.0035 -0.9997>}
 calf_twist_01_r {parent=calf_r, loc=<35.813 0.0967 -0.0397>, rot=<0.0237 0.0095 -0.0035 -0.9997>}
 deform_knee_in_base_r {parent=calf_r, loc=<-1.298 5.2957 0.0496>, rot=<0 0 0 -1>}
 deform_knee_in_r {parent=deform_knee_in_base_r, loc=<0 0 0>, rot=<0 0 0 1>}
 deform_knee_out_base_r {parent=calf_r, loc=<0 -4 0>, rot=<0 0 0 -1>}
 deform_knee_out_r {parent=deform_knee_out_base_r, loc=<0 0 0>, rot=<0 0 0 1>}
 thigh_twist_01_r {parent=thigh_r, loc=<21.6303 0 0>, rot=<0.0013 0 0 -1>}
 deform_groin_base_r {parent=thigh_r, loc=<2.1908 -8.1142 1.7442>, rot=<0 0 0 -1>}
 deform_groin_r {parent=deform_groin_base_r, loc=<0 0 0>, rot=<0 0 0 1>}
 deform_glute_base_l {parent=pelvis, loc=<-8.8064 -8.8222 -8.4287>, rot=<0.0275 0.0501 -0.4472 -0.8926>}
 deform_glute_l {parent=deform_glute_base_l, loc=<0 0 0>, rot=<0 0 0 1>}
 deform_glute_base_r {parent=pelvis, loc=<-8.8062 -8.8222 8.4287>, rot=<0.0501 -0.0275 0.8926 -0.4472>}
 deform_glute_r {parent=deform_glute_base_r, loc=<0 0 0>, rot=<0 0 0 1>}
 base_beltBuckle {parent=pelvis, loc=<5.3553 10.1679 0.2072>, rot=<0.5053 0.4997 0.4952 0.4997>}
 dyn_beltBuckle {parent=base_beltBuckle, loc=<0 0 0>, rot=<0 0 0 1>}
 base_phone {parent=pelvis, loc=<7.0837 9.2492 -7.0512>, rot=<-0.4285 -0.5669 -0.4185 -0.5655>}
 dyn_phone {parent=base_phone, loc=<0 0 0>, rot=<0 0 0 1>}
 base_belt {parent=pelvis, loc=<-0.7779 9.6776 0>, rot=<0.5053 0.4997 0.4952 0.4997>}
 dyn_belt {parent=base_belt, loc=<0 0 0>, rot=<0 0 0 1>}
 base_belt_l {parent=pelvis, loc=<-3.6784 8.6282 -8.1237>, rot=<0.5053 0.4997 0.4952 0.4997>}
 dyn_belt_l {parent=base_belt_l, loc=<0 0 0>, rot=<0 0 0 1>}
 base_belt_r {parent=pelvis, loc=<1.4089 8.5621 6.8451>, rot=<0.5053 0.4997 0.4952 0.4997>}
 dyn_belt_r {parent=base_belt_r, loc=<0 0 0>, rot=<0 0 0 1>}
 base_pen {parent=pelvis, loc=<-8.4117 5.5968 -15.9674>, rot=<-0.2473 -0.6457 -0.2963 -0.6589>}
 dyn_pen {parent=base_pen, loc=<0 0 0>, rot=<0 0 0 1>}
 base_frontJacket_l {parent=pelvis, loc=<8.7653 5.4253 -11.8326>, rot=<0.0366 -0.9382 -0.3384 -0.0617>}
 dyn_frontJacket_l_00 {parent=base_frontJacket_l, loc=<0 0 0>, rot=<0 0 0 1>}
 dyn_frontJacket_l_01 {parent=dyn_frontJacket_l_00, loc=<6 0 0>, rot=<0 0 0 1>}
 dyn_frontJacket_l_02 {parent=dyn_frontJacket_l_01, loc=<10.0238 0.1022 0.265>, rot=<-0.0043 0.0211 -0.0119 -0.9997>}
 dyn_frontJacket_l_03 {parent=dyn_frontJacket_l_02, loc=<11.4939 -0.1017 0.189>, rot=<0.0018 -0.0125 -0.0266 -0.9996>}
 dyn_frontJacket_l_04 {parent=dyn_frontJacket_l_03, loc=<14.2339 0.0147 0.0638>, rot=<0 0 0.0093 -1>}
 dyn_frontJacket_l_05 {parent=dyn_frontJacket_l_04, loc=<12.7977 -0.0419 0.5041>, rot=<0 0 0.0107 -0.9999>}
 base_midJacket_l {parent=pelvis, loc=<8.5739 -4.1312 -12.6786>, rot=<0.0061 -0.9899 0.107 -0.0929>}
 dyn_midJacket_l_00 {parent=base_midJacket_l, loc=<0 0 0>, rot=<0 0 0 1>}
 dyn_midJacket_l_01 {parent=dyn_midJacket_l_00, loc=<6 0 0>, rot=<0 0 0 1>}
 dyn_midJacket_l_02 {parent=dyn_midJacket_l_01, loc=<9.9993 0 0.0016>, rot=<0 -0.0062 0 -1>}
 dyn_midJacket_l_03 {parent=dyn_midJacket_l_02, loc=<11.4516 0 -0.2611>, rot=<0 -0.0291 0 -0.9996>}
 dyn_midJacket_l_04 {parent=dyn_midJacket_l_03, loc=<14.5154 0 0.3304>, rot=<0.0038 0.0108 0 -0.9999>}
 dyn_midJacket_l_05 {parent=dyn_midJacket_l_04, loc=<12.9082 0 -0.1411>, rot=<0 0 0 1>}
 base_backJacket_l {parent=pelvis, loc=<8.3937 -8.4207 -7.8631>, rot=<0.0313 -0.848 0.5228 -0.0817>}
 dyn_backJacket_l_00 {parent=base_backJacket_l, loc=<0 0 0>, rot=<0 0 0 1>}
 dyn_backJacket_l_01 {parent=dyn_backJacket_l_00, loc=<6 0 0>, rot=<0 0 0 1>}
 dyn_backJacket_l_02 {parent=dyn_backJacket_l_01, loc=<9.8437 -0.0962 -0.5231>, rot=<0 0 0 1>}
 dyn_backJacket_l_03 {parent=dyn_backJacket_l_02, loc=<11.3159 -0.1259 -0.6593>, rot=<0 -0.0093 0 -1>}
 dyn_backJacket_l_04 {parent=dyn_backJacket_l_03, loc=<14.2935 -0.0296 -0.3426>, rot=<0 0 0 1>}
 dyn_backJacket_l_05 {parent=dyn_backJacket_l_04, loc=<12.8127 -0.002 -0.2145>, rot=<0 0 0 1>}
 base_backJacket {parent=pelvis, loc=<8.2215 -8.9885 -0.1995>, rot=<0.0896 -0.7002 0.7049 -0.0691>}
 dyn_backJacket_00 {parent=base_backJacket, loc=<0 0 0>, rot=<0 0 0 1>}
 dyn_backJacket_01 {parent=dyn_backJacket_00, loc=<6 0 0>, rot=<0 0 0 1>}
 dyn_backJacket_02 {parent=dyn_backJacket_01, loc=<9.7767 0.1803 -0.8205>, rot=<0 0 0 1>}
 dyn_backJacket_03 {parent=dyn_backJacket_02, loc=<11.2331 0.2061 -1.0632>, rot=<0 -0.0414 0 -0.9991>}
 dyn_backJacket_04 {parent=dyn_backJacket_03, loc=<14.2137 0.268 -0.521>, rot=<0 0.0157 0 -0.9999>}
 dyn_backJacket_05 {parent=dyn_backJacket_04, loc=<12.7182 0.2393 -0.5261>, rot=<0 0 0 1>}
 base_backJacket_r {parent=pelvis, loc=<8.4258 -8.6046 7.9551>, rot=<0.067 -0.5232 0.8493 -0.0223>}
 dyn_backJacket_r_00 {parent=base_backJacket_r, loc=<0 0 0>, rot=<0 0 0 -1>}
 dyn_backJacket_r_01 {parent=dyn_backJacket_r_00, loc=<6 0 0>, rot=<0 0 0 1>}
 dyn_backJacket_r_02 {parent=dyn_backJacket_r_01, loc=<9.8558 0.0962 -0.1815>, rot=<0 0.0017 0 -1>}
 dyn_backJacket_r_03 {parent=dyn_backJacket_r_02, loc=<11.331 0.1259 -0.306>, rot=<0 0 0 -1>}
 dyn_backJacket_r_04 {parent=dyn_backJacket_r_03, loc=<14.2966 0.0295 -0.1639>, rot=<0 0 0 -1>}
 dyn_backJacket_r_05 {parent=dyn_backJacket_r_04, loc=<12.8119 0.002 0.2556>, rot=<0 0 0 -1>}
 base_midJacket_r {parent=pelvis, loc=<8.5838 -4.1427 12.7308>, rot=<0.0884 -0.107 0.9903 -0.0056>}
 dyn_midJacket_r_00 {parent=base_midJacket_r, loc=<0 0 0>, rot=<0 0 0 1>}
 dyn_midJacket_r_01 {parent=dyn_midJacket_r_00, loc=<6 0 0>, rot=<0 0 0 1>}
 dyn_midJacket_r_02 {parent=dyn_midJacket_r_01, loc=<9.9986 0 -0.1219>, rot=<0 0 0 -1>}
 dyn_midJacket_r_03 {parent=dyn_midJacket_r_02, loc=<11.4481 0 -0.3845>, rot=<0 0 0 -1>}
 dyn_midJacket_r_04 {parent=dyn_midJacket_r_03, loc=<14.5106 0 -0.5005>, rot=<0 0 0 -1>}
 dyn_midJacket_r_05 {parent=dyn_midJacket_r_04, loc=<12.9024 0 -0.4078>, rot=<0 0 0 -1>}
 base_frontJacket_r {parent=pelvis, loc=<8.7333 5.2462 11.62>, rot=<0.0836 0.3392 0.9366 -0.0287>}
 dyn_frontJacket_r_00 {parent=base_frontJacket_r, loc=<0 0 0>, rot=<0 0 0 1>}
 dyn_frontJacket_r_01 {parent=dyn_frontJacket_r_00, loc=<6 0 0>, rot=<0 0 0 1>}
 dyn_frontJacket_r_02 {parent=dyn_frontJacket_r_01, loc=<10.0252 -0.1022 -0.2026>, rot=<0.0037 -0.0236 0.012 -0.9996>}
 dyn_frontJacket_r_03 {parent=dyn_frontJacket_r_02, loc=<11.4752 0.1065 0.681>, rot=<-0.0018 -0.013 0.0266 -0.9996>}
 dyn_frontJacket_r_04 {parent=dyn_frontJacket_r_03, loc=<14.2175 -0.0076 0.6882>, rot=<0 0 -0.0093 -1>}
 dyn_frontJacket_r_05 {parent=dyn_frontJacket_r_04, loc=<12.7633 0.0484 1.0653>, rot=<0 0 -0.0107 -0.9999>}
 dyn_Cloth_volume_bk {parent=pelvis, loc=<12.4262 -13.7579 0.3385>, rot=<0 0 -0.0151 -0.9999>}
 dyn_Cloth_volume_bk01 {parent=dyn_Cloth_volume_bk, loc=<-6.1856 0 0>, rot=<0 0 0 1>}
 dyn_Cloth_volume_bk02 {parent=dyn_Cloth_volume_bk01, loc=<-11.8523 0 0>, rot=<0.7055 -0.0406 0.7073 0.0195>}
 dyn_Cloth_volume_bk03 {parent=dyn_Cloth_volume_bk02, loc=<0 0 -12.173>, rot=<0 0 0 1>}
 dyn_Cloth_volume_bk04 {parent=dyn_Cloth_volume_bk03, loc=<0 0 -17.3458>, rot=<0 0 0 1>}
 dyn_Cloth_volume_l_B {parent=pelvis, loc=<12.5408 -0.8485 -15.6013>, rot=<-0.7089 0.046 -0.0204 -0.7035>}
 dyn_Cloth_volume_l_B01 {parent=dyn_Cloth_volume_l_B, loc=<-6.9564 0 0>, rot=<0 0 0 1>}
 dyn_Cloth_volume_l_B02 {parent=dyn_Cloth_volume_l_B01, loc=<-13.0905 0 0>, rot=<0 0 0 1>}
 dyn_Cloth_volume_l_B03 {parent=dyn_Cloth_volume_l_B02, loc=<-13.2738 0 0>, rot=<0 0 0 1>}
 dyn_Cloth_volume_l_B04 {parent=dyn_Cloth_volume_l_B03, loc=<-14.2487 0 0>, rot=<0 0 0 1>}
 dyn_Cloth_volume_l_C {parent=pelvis, loc=<12.83 -11.3376 -10.8975>, rot=<-0.1378 0.007 -0.0149 -0.9903>}
 dyn_Cloth_volume_l_C01 {parent=dyn_Cloth_volume_l_C, loc=<-5.9349 0 0>, rot=<0 0 0 1>}
 dyn_Cloth_volume_l_C02 {parent=dyn_Cloth_volume_l_C01, loc=<-12.3087 0 0>, rot=<0 0 0 1>}
 dyn_Cloth_volume_l_C03 {parent=dyn_Cloth_volume_l_C02, loc=<-14.423 0 0>, rot=<0 0 0 1>}
 dyn_Cloth_volume_l_C04 {parent=dyn_Cloth_volume_l_C03, loc=<-14.6821 0 0>, rot=<0 0 0 1>}
 dyn_Cloth_volume_r_B {parent=pelvis, loc=<12.5285 -1.1911 15.9591>, rot=<0.7085 -0.0521 -0.0264 -0.7033>}
 dyn_Cloth_volume_r_B01 {parent=dyn_Cloth_volume_r_B, loc=<-6.9777 0 0>, rot=<0 0 0 1>}
 dyn_Cloth_volume_r_B02 {parent=dyn_Cloth_volume_r_B01, loc=<-13.1229 0 0>, rot=<0 0 0 1>}
 dyn_Cloth_volume_r_B03 {parent=dyn_Cloth_volume_r_B02, loc=<-13.2738 0 0>, rot=<0 0 0 1>}
 dyn_Cloth_volume_r_B04 {parent=dyn_Cloth_volume_r_B03, loc=<-14.2487 0 0>, rot=<0 0 0 1>}
 dyn_Cloth_volume_r_C {parent=pelvis, loc=<12.9366 -11.4322 10.6309>, rot=<0.1377 -0.0088 -0.0277 -0.9901>}
 dyn_Cloth_volume_r_C01 {parent=dyn_Cloth_volume_r_C, loc=<-5.9091 0 0>, rot=<0 0 0 1>}
 dyn_Cloth_volume_r_C02 {parent=dyn_Cloth_volume_r_C01, loc=<-12.2974 0 0>, rot=<0 0 0 1>}
 dyn_Cloth_volume_r_C03 {parent=dyn_Cloth_volume_r_C02, loc=<-14.423 0 0>, rot=<0 0 0 1>}
 dyn_Cloth_volume_r_C04 {parent=dyn_Cloth_volume_r_C03, loc=<-14.6821 0 0>, rot=<0 0 0 1>}
 dyn_Cloth_volume_l_A {parent=pelvis, loc=<11.1469 11.3866 -4.7174>, rot=<0.9889 -0.0137 -0.0032 0.1479>}
 dyn_Cloth_volume_l_A01 {parent=dyn_Cloth_volume_l_A, loc=<-6.9566 0 0>, rot=<0 0 0 1>}
 dyn_Cloth_volume_l_A02 {parent=dyn_Cloth_volume_l_A01, loc=<-12.2733 0 0>, rot=<-0.002 -0.0542 -0.0376 -0.9978>}
 dyn_Cloth_volume_l_A03 {parent=dyn_Cloth_volume_l_A02, loc=<-14.0317 0 0>, rot=<0 0 0 1>}
 dyn_Cloth_volume_l_A04 {parent=dyn_Cloth_volume_l_A03, loc=<-10.4224 0 0>, rot=<0 0 0 1>}
 dyn_Cloth_volume_r_A {parent=pelvis, loc=<11.1496 11.3865 4.7174>, rot=<0.9889 -0.0137 0.0032 -0.1479>}
 dyn_Cloth_volume_r_A01 {parent=dyn_Cloth_volume_r_A, loc=<-6.9473 0 0>, rot=<0 0 0.0286 -0.9996>}
 dyn_Cloth_volume_r_A02 {parent=dyn_Cloth_volume_r_A01, loc=<-12.3245 0 0>, rot=<-0.0042 0.0538 -0.0771 -0.9956>}
 dyn_Cloth_volume_r_A03 {parent=dyn_Cloth_volume_r_A02, loc=<-14.0317 0 0>, rot=<0 0 0 1>}
 dyn_Cloth_volume_r_A04 {parent=dyn_Cloth_volume_r_A03, loc=<-10.4224 0 0>, rot=<0 0 0 1>}
 gravity {parent=root, loc=<0 0 0>, rot=<0 0 0 1>}
 gravity_cloth {parent=root, loc=<0 0 0>, rot=<0 0 0 1>}
]
variables [
 Leg_l_Swing_rot {type=FVector, default=(X=0,Y=0,Z=0)}
 Leg_r_Swing_rot {type=FVector, default=(X=0,Y=0,Z=0)}
]
