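# Physics asset for the upper-body mesh of the medium male CrawlyTech player cosmetic, set up for a dangling tie and nametag.
# Capsules on spine, arms and thighs, plus belt buckle and phone shapes, form the body volumes.
# A six-segment tie chain hangs from spine_05 with swing limits widening down the chain, and a boxed nametag hangs from spine_04.
# Body and anchor constraints swing and twist freely; thighs, buckle and phone have no constraints.
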PhysicsAsset /BRCosmetics/Characters/Player/Male/Medium/Bodies/M_MED_CrawlyTech/Meshes/M_MED_CrawlyTech_Upper_Body_Physics
bodies {
 base_beltBuckle [{shapeType=Box, extentX=8, extentY=1.5, extentZ=5, shapeName=base_beltBuckle_box, rotation=<rot 0 -90 0>}]
 base_nametag [{shapeType=Capsule, length=0.5, radius=0.5, shapeName=base_nametag_capsule}]
 base_phone [{shapeType=Capsule, length=10, radius=4.5, shapeName=base_phone_capsule, center=<1.4872 -0.6739 -1.653>, rotation=<rot 0 -90 0>}]
 base_tie [{shapeType=Capsule, length=0.5, radius=0.5, shapeName=base_tie_capsule}]
 dyn_nametag [{
   shapeType=Box
   extentX=0.7495
   extentY=9.1389
   extentZ=6.2666
   shapeName=dyn_nametag_box
   center=<-0.0018 -0.0011 -5.0506>
   rotation=<rot -12.9043 19.5809 0.2365>
  }]
 dyn_tie_01 [{shapeType=Capsule, length=3.5, radius=1, shapeName=dyn_tie_01_capsule, center=<3 0 0>, rotation=<rot 90 0 0>}]
 dyn_tie_02 [{shapeType=Capsule, length=4.2505, radius=1.2173, shapeName=dyn_tie_02_capsule, center=<3 0 0.0485>, rotation=<rot 90 0 0>}]
 dyn_tie_03 [{shapeType=Capsule, length=4.2505, radius=1.2173, shapeName=dyn_tie_03_capsule, center=<3 0 0.0485>, rotation=<rot 90 0 0>}]
 dyn_tie_04 [{shapeType=Capsule, length=4.2505, radius=1.2173, shapeName=dyn_tie_04_capsule, center=<3 0 0.0485>, rotation=<rot 90 0 0>}]
 dyn_tie_05 [{shapeType=Capsule, length=4.2505, radius=1.2173, shapeName=dyn_tie_05_capsule, center=<3 0 0.0485>, rotation=<rot 90 0 0>}]
 dyn_tie_06 [{shapeType=Capsule, length=2.8369, radius=1.2173, shapeName=dyn_tie_06_capsule, center=<3 0 0.0485>, rotation=<rot 90 0 0>}]
 lowerarm_l [{shapeType=Capsule, length=16.031, radius=4.969, shapeName=lowerarm_l_capsule, center=<8 0 0>, rotation=<rot 90 0 0>}]
 lowerarm_r [{shapeType=Capsule, length=16.031, radius=4.969, shapeName=lowerarm_r_capsule, center=<-8 0 0>, rotation=<rot 90 180 0>}]
 spine_01 [{shapeType=Capsule, length=10, radius=10.5, shapeName=spine_01_capsule, center=<2 0 0>}]
 spine_02 [{shapeType=Capsule, length=9.6329, radius=10.8671, shapeName=spine_02_capsule, center=<2.5027 -0.1515 0>}]
 spine_03 [{shapeType=Capsule, length=10, radius=10.5, shapeName=spine_03_capsule, center=<2 0.1524 0>}]
 spine_04 [
  {shapeType=Capsule, length=0, radius=10.9817, shapeName=spine_04_capsule, center=<2 -0.2098 0>}
  {
   shapeType=Capsule
   length=2.8386
   radius=9.6088
   shapeName=spine_04_capsule2
   center=<1.4088 0.4831 -4.0917>
   rotation=<rot 49.2648 87.0579 83.3946>
  }
  {
   shapeType=Capsule
   length=7.7898
   radius=4.6577
   shapeName=spine_04_capsule3
   center=<1.2579 6.3479 -3.2771>
   rotation=<rot 49.2648 87.0579 83.3946>
  }
 ]
 spine_05 [{shapeType=Capsule, length=10, radius=8, shapeName=spine_05_capsule, center=<0.8565 0.9982 0>}]
 thigh_l [{shapeType=Capsule, length=26.364, radius=9.166, shapeName=thigh_l_capsule, center=<-21.6064 0 0>, rotation=<rot 90 0 0>}]
 thigh_r [{shapeType=Capsule, length=26.364, radius=9.166, shapeName=thigh_r_capsule, center=<21.6064 0 0>, rotation=<rot 90 0 180>}]
 upperarm_l [{shapeType=Capsule, length=16.1112, radius=4.8888, shapeName=upperarm_l_capsule, center=<8 0 -1.6211>, rotation=<rot 90 0 0>}]
 upperarm_r [{shapeType=Capsule, length=16.1112, radius=4.8888, shapeName=upperarm_r_capsule, center=<-8 0 1.6211>, rotation=<rot 90 0 180>}]
}
constraints [
 {
  bone1Name=spine_02
  bone2Name=spine_01
  swing1LimitDegrees=45
  swing1Motion=Free
  swing2LimitDegrees=45
  swing2Motion=Free
  twistLimitDegrees=45
  twistMotion=Free
 }
 {
  bone1Name=spine_03
  bone2Name=spine_02
  swing1LimitDegrees=45
  swing1Motion=Free
  swing2LimitDegrees=45
  swing2Motion=Free
  twistLimitDegrees=45
  twistMotion=Free
 }
 {
  bone1Name=spine_04
  bone2Name=spine_03
  swing1LimitDegrees=45
  swing1Motion=Free
  swing2LimitDegrees=45
  swing2Motion=Free
  twistLimitDegrees=45
  twistMotion=Free
 }
 {
  bone1Name=spine_05
  bone2Name=spine_04
  swing1LimitDegrees=45
  swing1Motion=Free
  swing2LimitDegrees=45
  swing2Motion=Free
  twistLimitDegrees=45
  twistMotion=Free
 }
 {
  bone1Name=upperarm_l
  bone2Name=spine_05
  swing1LimitDegrees=45
  swing1Motion=Free
  swing2LimitDegrees=45
  swing2Motion=Free
  twistLimitDegrees=45
  twistMotion=Free
 }
 {
  bone1Name=lowerarm_l
  bone2Name=upperarm_l
  swing1LimitDegrees=45
  swing1Motion=Free
  swing2LimitDegrees=45
  swing2Motion=Free
  twistLimitDegrees=45
  twistMotion=Free
 }
 {
  bone1Name=upperarm_r
  bone2Name=spine_05
  swing1LimitDegrees=45
  swing1Motion=Free
  swing2LimitDegrees=45
  swing2Motion=Free
  twistLimitDegrees=45
  twistMotion=Free
 }
 {
  bone1Name=lowerarm_r
  bone2Name=upperarm_r
  swing1LimitDegrees=45
  swing1Motion=Free
  swing2LimitDegrees=45
  swing2Motion=Free
  twistLimitDegrees=45
  twistMotion=Free
 }
 {
  bone1Name=base_tie
  bone2Name=spine_05
  swing1LimitDegrees=45
  swing1Motion=Free
  swing2LimitDegrees=45
  swing2Motion=Free
  twistLimitDegrees=45
  twistMotion=Free
 }
 {bone1Name=dyn_tie_01, bone2Name=base_tie, swing1LimitDegrees=5, swing2LimitDegrees=5, twistLimitDegrees=0}
 {bone1Name=dyn_tie_02, bone2Name=dyn_tie_01, swing1LimitDegrees=10, swing2LimitDegrees=10, twistLimitDegrees=0}
 {bone1Name=dyn_tie_03, bone2Name=dyn_tie_02, swing1LimitDegrees=15, swing2LimitDegrees=15, twistLimitDegrees=0}
 {bone1Name=dyn_tie_04, bone2Name=dyn_tie_03, swing1LimitDegrees=20, swing2LimitDegrees=20, twistLimitDegrees=0}
 {bone1Name=dyn_tie_05, bone2Name=dyn_tie_04, swing1LimitDegrees=25, swing2LimitDegrees=25, twistLimitDegrees=0}
 {bone1Name=dyn_tie_06, bone2Name=dyn_tie_05, swing1LimitDegrees=25, swing2LimitDegrees=25, twistLimitDegrees=0}
 {
  bone1Name=base_nametag
  bone2Name=spine_04
  swing1LimitDegrees=45
  swing1Motion=Free
  swing2LimitDegrees=45
  swing2Motion=Free
  twistLimitDegrees=45
  twistMotion=Free
 }
 {bone1Name=dyn_nametag, bone2Name=base_nametag, swing1LimitDegrees=25, swing2LimitDegrees=25, twistLimitDegrees=0}
]
