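# Post-process style AnimBP for the CrawlyTech medium male body, adding secondary motion to its coat.
# The copied parent pose runs through a Control Rig, gravity-bone constraints and jacket pre-sim offsets.
# Upper-body RigidBody simulation feeds a bool blend of standard and crouched coat simulations, then post-sim jacket corrections.
# The EventGraph derives movement flags and per-state gravity, wind and RigidBody settings through a nativized gravity-override function.
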AnimBlueprint /BRCosmetics/Characters/Player/Male/Medium/Bodies/M_MED_CrawlyTech/Meshes/M_MED_CrawlyTech_AnimBP
target_skeleton=&/M_MED_CrawlyTech_Skeleton
graphs [
 AnimGraph {nodes [
   RigidBody!(ComponentPose, Alpha, OverrideWorldGravity:Vector="0,0,0") -> (Pose) {
    bClampLinearTranslationLimitToRefPose=true
    bOverrideWorldGravity=true
    overridePhysicsAsset=&/M_MED_CrawlyTech_Coat_Physics
    simulationSpace=BaseBoneSpace
    alphaBoolBlend {blendOption=Linear}
    baseBoneRef {boneName=pelvis}
    simSpaceSettings {
     maxAngularAcceleration=1600
     maxAngularVelocity=4
     maxLinearAcceleration=1000
     maxLinearVelocity=400
     velocityScaleZ=0.75
     worldAlpha=1
     externalLinearDragV=<1.1 1.1 1.1>
    }
   }
   RigidBody!(
    ComponentPose
    Alpha
    OverrideWorldGravity="0,0,0"
    ExternalForce:Vector
    SimSpaceSettings:Sim Space Settings Structure=(WorldAlpha=0,VelocityScaleZ=1,DampingAlpha=1,MaxLinearVelocity=10000,MaxAngularVelocity=10000,MaxLinearAcceleration=10000,MaxAngularAcceleration=10000,ExternalLinearDragV=(X=0,Y=0,Z=0),ExternalLinearVelocity=(X=0,Y=0,Z=0),ExternalAngularVelocity=(X=0,Y=0,Z=0))
   ) -> (Pose) {
    bOverrideWorldGravity=true
    overridePhysicsAsset=&M_MED_CrawlyTech_Coat_Physics
    simulationSpace=BaseBoneSpace
    alphaBoolBlend {blendOption=Linear}
    baseBoneRef {boneName=gravity}
    componentAppliedLinearAccClamp=<0 0 0>
   }
   n1 = "Transform (Modify) Bone - Bone: base_frontJacket_l"!(
    ComponentPose=Local To Component(LocalPose=Use cached pose 'dynamicsCP'())
    bAlphaBoolEnabled:bool=True
    Translation="0,0,-2"
    Rotation="15,0,0"
   ) -> (Pose) {
    alphaInputType=Bool
    rotationMode=BMM_Additive
    rotationSpace=BCS_BoneSpace
    translationMode=BMM_Additive
    translationSpace=BCS_BoneSpace
    alphaBoolBlend {blendInTime=0.1, blendOption=Linear, blendOutTime=0.1}
    alphaScaleBiasClamp {bInterpResult=true}
    boneToModify {boneName=base_frontJacket_l}
    rotation {pitch=15}
    translation {z=-2}
   }
   n2 = "Transform (Modify) Bone - Bone: base_frontJacket_r"!(ComponentPose=@n1, bAlphaBoolEnabled:bool=True, Translation="0,0,1.5", Rotation="12,0,0") -> (Pose) {
    alphaInputType=Bool
    rotationMode=BMM_Additive
    rotationSpace=BCS_BoneSpace
    translationMode=BMM_Additive
    translationSpace=BCS_BoneSpace
    alphaBoolBlend {blendInTime=0.1, blendOption=Linear, blendOutTime=0.1}
    alphaScaleBiasClamp {bInterpResult=true}
    boneToModify {boneName=base_frontJacket_r}
    rotation {pitch=12}
    translation {z=1.5}
   }
   n3 = "Transform (Modify) Bone - Bone: base_midJacket_r"!(ComponentPose=@n2, bAlphaBoolEnabled:bool=True, Translation="0,0,1", Rotation="10,0,0") -> (Pose) {
    alphaInputType=Bool
    rotationMode=BMM_Additive
    rotationSpace=BCS_BoneSpace
    translationMode=BMM_Additive
    translationSpace=BCS_BoneSpace
    alphaBoolBlend {blendInTime=0.1, blendOption=Linear, blendOutTime=0.1}
    alphaScaleBiasClamp {bInterpResult=true}
    boneToModify {boneName=base_midJacket_r}
    rotation {pitch=10}
    translation {z=1}
   }
   n4 = "Transform (Modify) Bone - Bone: base_frontJacket_l"(ComponentPose=@n3, bAlphaBoolEnabled=True, Translation="0,0,2", Rotation="-10,0,0") {
    alphaInputType=Bool
    rotationMode=BMM_Additive
    rotationSpace=BCS_BoneSpace
    translationMode=BMM_Additive
    translationSpace=BCS_BoneSpace
    alphaBoolBlend {blendInTime=0.1, blendOption=Linear, blendOutTime=0.1}
    boneToModify {boneName=base_frontJacket_l}
    rotation {pitch=-10}
    translation {z=2}
   }
   n5 = "Transform (Modify) Bone - Bone: base_midJacket_l"!(ComponentPose=@n4, bAlphaBoolEnabled:bool=True, Translation="-2,0,-2", Rotation="9,0,0") -> (Pose) {
    alphaInputType=Bool
    rotationMode=BMM_Additive
    rotationSpace=BCS_BoneSpace
    translationMode=BMM_Additive
    translationSpace=BCS_BoneSpace
    alphaBoolBlend {blendInTime=0.1, blendOption=Linear, blendOutTime=0.1}
    boneToModify {boneName=base_midJacket_l}
    rotation {pitch=9}
    translation {x=-2, z=-2}
   }
   n6 = "Transform (Modify) Bone - Bone: base_frontJacket_r"(ComponentPose=@n5, bAlphaBoolEnabled=True, Translation="0,0,-2", Rotation="-10,0,0") {
    alphaInputType=Bool
    rotationMode=BMM_Additive
    rotationSpace=BCS_BoneSpace
    translationMode=BMM_Additive
    translationSpace=BCS_BoneSpace
    alphaBoolBlend {blendInTime=0.1, blendOption=Linear, blendOutTime=0.1}
    boneToModify {boneName=base_frontJacket_r}
    rotation {pitch=-10}
    translation {z=-2}
   }
   n7 = "Transform (Modify) Bone - Bone: dyn_backJacket_r_03"!(ComponentPose=@n6, bAlphaBoolEnabled:bool=True, Translation="0,0,0", Rotation="10,0,0") -> (Pose) {
    alphaInputType=Bool
    rotationMode=BMM_Additive
    rotationSpace=BCS_BoneSpace
    translationMode=BMM_Additive
    translationSpace=BCS_BoneSpace
    alphaBoolBlend {blendInTime=0.1, blendOption=Linear, blendOutTime=0.1}
    boneToModify {boneName=dyn_backJacket_r_03}
    rotation {pitch=10}
   }
   n8 = "Transform (Modify) Bone - Bone: base_frontJacket_l"(ComponentPose=@n7, bAlphaBoolEnabled=True, Translation="0,-1,1.5", Rotation="0,0,0") {
    alphaInputType=Bool
    rotationMode=BMM_Additive
    rotationSpace=BCS_BoneSpace
    translationMode=BMM_Additive
    translationSpace=BCS_BoneSpace
    alphaBoolBlend {blendInTime=0.5, blendOption=Cubic, blendOutTime=0.5}
    boneToModify {boneName=base_frontJacket_l}
    translation {y=-1, z=1.5}
   }
   n9 = "Transform (Modify) Bone - Bone: dyn_backJacket_01"!(ComponentPose=@n8, bAlphaBoolEnabled:bool=True, Translation="0,0,0", Rotation="7,0,0") -> (Pose) {
    alphaInputType=Bool
    rotationMode=BMM_Additive
    rotationSpace=BCS_BoneSpace
    translationMode=BMM_Additive
    translationSpace=BCS_BoneSpace
    alphaBoolBlend {blendInTime=0.5, blendOption=Cubic, blendOutTime=0.5}
    boneToModify {boneName=dyn_backJacket_01}
    rotation {pitch=7}
   }
   Output Pose(Result=Component To Local(ComponentPose=@n9)) {layerGroup=None}
   n10 = Control Rig(Source=Copy Pose From Mesh() {bUseAttachedParent=true}) {
    alphaBoolBlend {blendOption=Linear}
    controlRigAssetReference {blueprintRigClass=&/M_MED_CrawlyTech_CtrlRig.M_MED_CrawlyTech_CtrlRig_C}
   }
   Save cached pose 'controlRigCP'(Pose=@n10) {cacheName=controlRigCP}
   n11 = RigidBody(
    ComponentPose=Local To Component(LocalPose=Use cached pose 'DynamicsS1CP'())
    OverrideWorldGravity="0,0,0"
    SimSpaceSettings=(WorldAlpha=0,VelocityScaleZ=1,DampingAlpha=1,MaxLinearVelocity=10000,MaxAngularVelocity=10000,MaxLinearAcceleration=10000,MaxAngularAcceleration=10000,ExternalLinearDragV=(X=0,Y=0,Z=0),ExternalLinearVelocity=(X=0,Y=0,Z=0),ExternalAngularVelocity=(X=0,Y=0,Z=0))
   ) {
    bOverrideWorldGravity=true
    overridePhysicsAsset=&/M_MED_CrawlyTech_Coat_Physics_Crouched
    simulationSpace=BaseBoneSpace
    alphaBoolBlend {blendOption=Linear}
    baseBoneRef {boneName=gravity}
    componentAppliedLinearAccClamp=<0 0 0>
   }
   n12 = RigidBody!(
    ComponentPose=Local To Component(LocalPose=Use cached pose 'DynamicsS1CP'())
    Alpha
    OverrideWorldGravity="0,0,0"
    SimSpaceSettings=(WorldAlpha=1,VelocityScaleZ=0.75,DampingAlpha=1,MaxLinearVelocity=400,MaxAngularVelocity=4,MaxLinearAcceleration=1000,MaxAngularAcceleration=1600,ExternalLinearDragV=(X=1.1,Y=1.1,Z=1.1),ExternalLinearVelocity=(X=0,Y=0,Z=0),ExternalAngularVelocity=(X=0,Y=0,Z=0))
   ) -> (Pose) {
    bClampLinearTranslationLimitToRefPose=true
    bOverrideWorldGravity=true
    overridePhysicsAsset=&M_MED_CrawlyTech_Coat_Physics
    simulationSpace=BaseBoneSpace
    alphaBoolBlend {blendOption=Linear}
    baseBoneRef {boneName=pelvis}
    simSpaceSettings {
     maxAngularAcceleration=1600
     maxAngularVelocity=4
     maxLinearAcceleration=1000
     maxLinearVelocity=400
     velocityScaleZ=0.75
     worldAlpha=1
     externalLinearDragV=<1.1 1.1 1.1>
    }
   }
   Save cached pose 'dynamicsCP'(Pose=Blend Poses by bool(BlendPose_0=Component To Local(ComponentPose=@n11), BlendPose_1=Component To Local(ComponentPose=@n12)) {blendType=Linear}) {cacheName=dynamicsCP}
   n13 = RigidBody!(
    ComponentPose=Local To Component(LocalPose=Use cached pose 'presimFixesCP'())
    Alpha
    OverrideWorldGravity="0,0,0"
    ExternalForce:Vector
    SimSpaceSettings=(WorldAlpha=0,VelocityScaleZ=1,DampingAlpha=1,MaxLinearVelocity=10000,MaxAngularVelocity=10000,MaxLinearAcceleration=10000,MaxAngularAcceleration=10000,ExternalLinearDragV=(X=0,Y=0,Z=0),ExternalLinearVelocity=(X=0,Y=0,Z=0),ExternalAngularVelocity=(X=0,Y=0,Z=0))
   ) -> (Pose) {
    bOverrideWorldGravity=true
    overridePhysicsAsset=&/M_MED_CrawlyTech_Upper_Body_Physics
    simulationSpace=BaseBoneSpace
    alphaBoolBlend {blendOption=Linear}
    baseBoneRef {boneName=spine_05}
    componentAppliedLinearAccClamp=<3000 3000 3000>
   }
   Save cached pose 'DynamicsS1CP'(Pose=Component To Local(ComponentPose=@n13)) {cacheName=DynamicsS1CP}
   n14 = "Constraint - gravity (TT)"!(
    ComponentPose=Local To Component(LocalPose=Use cached pose 'controlRigCP'())
    Alpha=1
    ConstraintWeights_0=0.1
    ConstraintWeights_1:float=0.9
   ) -> (Pose) {
    alphaBoolBlend {blendOption=Linear}
    boneToModify {boneName=gravity}
    constraintSetup [{transformType=Translation, targetBone {boneName=root}}, {transformType=Translation, targetBone {boneName=pelvis}}]
    constraintWeights [1,1]
   }
   n15 = "Constraint - gravity (PP)"!(ComponentPose=@n14, Alpha=1, ConstraintWeights_0=1, ConstraintWeights_1:float=0) -> (Pose) {
    alphaBoolBlend {blendOption=Linear}
    boneToModify {boneName=gravity}
    constraintSetup [{targetBone {boneName=root}}, {targetBone {boneName=pelvis}}]
    constraintWeights [1,1]
   }
   n16 = "Transform (Modify) Bone - Bone: base_backJacket_l"!(ComponentPose=@n15, Alpha=1, Rotation="15,0,0") -> (Pose) {
    rotationMode=BMM_Additive
    rotationSpace=BCS_BoneSpace
    alphaBoolBlend {blendOption=Linear}
    alphaScaleBiasClamp {bInterpResult=true}
    boneToModify {boneName=base_backJacket_l}
    rotation {pitch=15}
   }
   n17 = "Transform (Modify) Bone - Bone: base_backJacket"!(ComponentPose=@n16, Alpha=1, Rotation="20,0,0") -> (Pose) {
    rotationMode=BMM_Additive
    rotationSpace=BCS_BoneSpace
    alphaBoolBlend {blendOption=Linear}
    alphaScaleBiasClamp {bInterpResult=true}
    boneToModify {boneName=base_backJacket}
    rotation {pitch=20}
   }
   n18 = "Transform (Modify) Bone - Bone: base_backJacket_r"!(ComponentPose=@n17, Alpha=1, Rotation="15,0,0") -> (Pose) {
    rotationMode=BMM_Additive
    rotationSpace=BCS_BoneSpace
    alphaBoolBlend {blendOption=Linear}
    alphaScaleBiasClamp {bInterpResult=true}
    boneToModify {boneName=base_backJacket_r}
    rotation {pitch=15}
   }
   n19 = "Transform (Modify) Bone - Bone: base_frontJacket_l"!(ComponentPose=@n18, Alpha:float=1, Rotation="40,0,0") -> (Pose) {
    rotationMode=BMM_Additive
    rotationSpace=BCS_BoneSpace
    alphaBoolBlend {blendOption=Linear}
    alphaScaleBiasClamp {bInterpResult=true}
    boneToModify {boneName=base_frontJacket_l}
    rotation {pitch=40}
   }
   n20 = "Transform (Modify) Bone - Bone: base_midJacket_l"!(ComponentPose=@n19, Alpha:float=1, Rotation="85,0,0") -> (Pose) {
    rotationMode=BMM_Additive
    rotationSpace=BCS_BoneSpace
    alphaBoolBlend {blendOption=Linear}
    alphaScaleBiasClamp {bInterpResult=true}
    boneToModify {boneName=base_midJacket_l}
    rotation {pitch=85}
   }
   n21 = "Transform (Modify) Bone - Bone: base_backJacket_l"(ComponentPose=@n20, Alpha=1, Rotation="100,0,0") {
    rotationMode=BMM_Additive
    rotationSpace=BCS_BoneSpace
    alphaBoolBlend {blendOption=Linear}
    alphaScaleBiasClamp {bInterpResult=true}
    boneToModify {boneName=base_backJacket_l}
    rotation {pitch=100}
   }
   n22 = "Transform (Modify) Bone - Bone: base_backJacket"(ComponentPose=@n21, Alpha=1, Rotation="95,0,0") {
    rotationMode=BMM_Additive
    rotationSpace=BCS_BoneSpace
    alphaBoolBlend {blendOption=Linear}
    alphaScaleBiasClamp {bInterpResult=true}
    boneToModify {boneName=base_backJacket}
    rotation {pitch=95}
   }
   n23 = "Transform (Modify) Bone - Bone: base_backJacket_r"(ComponentPose=@n22, Alpha=1, Rotation="100,0,0") {
    rotationMode=BMM_Additive
    rotationSpace=BCS_BoneSpace
    alphaBoolBlend {blendOption=Linear}
    alphaScaleBiasClamp {bInterpResult=true}
    boneToModify {boneName=base_backJacket_r}
    rotation {pitch=100}
   }
   n24 = "Transform (Modify) Bone - Bone: base_midJacket_r"!(ComponentPose=@n23, Alpha:float=1, Rotation="65,0,0") -> (Pose) {
    rotationMode=BMM_Additive
    rotationSpace=BCS_BoneSpace
    alphaBoolBlend {blendOption=Linear}
    alphaScaleBiasClamp {bInterpResult=true}
    boneToModify {boneName=base_midJacket_r}
    rotation {pitch=65}
   }
   n25 = "Transform (Modify) Bone - Bone: base_frontJacket_r"!(ComponentPose=@n24, Alpha:float=1, Rotation="35,0,0") -> (Pose) {
    rotationMode=BMM_Additive
    rotationSpace=BCS_BoneSpace
    alphaBoolBlend {blendOption=Linear}
    alphaScaleBiasClamp {bInterpResult=true}
    boneToModify {boneName=base_frontJacket_r}
    rotation {pitch=35}
   }
   n26 = "Transform (Modify) Bone - Bone: base_frontJacket_l"!(ComponentPose=@n25, bAlphaBoolEnabled:bool=True, Translation:Vector="0,0,0", Rotation="0,0,-45") -> (Pose) {
    alphaInputType=Bool
    rotationMode=BMM_Additive
    rotationSpace=BCS_BoneSpace
    translationMode=BMM_Additive
    translationSpace=BCS_BoneSpace
    alphaBoolBlend {blendOption=Linear}
    alphaScaleBiasClamp {bInterpResult=true}
    boneToModify {boneName=base_frontJacket_l}
    rotation {roll=-45}
   }
   n27 = "Transform (Modify) Bone - Bone: base_midJacket_l"!(ComponentPose=@n26, bAlphaBoolEnabled:bool=True, Translation:Vector="0,0,0", Rotation="0,5.671,0") -> (Pose) {
    alphaInputType=Bool
    rotationMode=BMM_Additive
    rotationSpace=BCS_BoneSpace
    translationMode=BMM_Additive
    translationSpace=BCS_BoneSpace
    alphaBoolBlend {blendOption=Linear}
    alphaScaleBiasClamp {bInterpResult=true}
    boneToModify {boneName=base_midJacket_l}
    rotation {yaw=5.671}
   }
   n28 = "Transform (Modify) Bone - Bone: base_backJacket_l"!(ComponentPose=@n27, bAlphaBoolEnabled:bool=True, Translation:Vector="0,0,2", Rotation="0,0,0") -> (Pose) {
    alphaInputType=Bool
    rotationMode=BMM_Additive
    rotationSpace=BCS_BoneSpace
    translationMode=BMM_Additive
    translationSpace=BCS_BoneSpace
    alphaBoolBlend {blendOption=Linear}
    alphaScaleBiasClamp {bInterpResult=true}
    boneToModify {boneName=base_backJacket_l}
    translation {z=2}
   }
   n29 = "Transform (Modify) Bone - Bone: base_backJacket"!(ComponentPose=@n28, bAlphaBoolEnabled:bool=True, Translation:Vector="0,0,2", Rotation="0,0,0") -> (Pose) {
    alphaInputType=Bool
    rotationMode=BMM_Additive
    rotationSpace=BCS_BoneSpace
    translationMode=BMM_Additive
    translationSpace=BCS_BoneSpace
    alphaBoolBlend {blendOption=Linear}
    alphaScaleBiasClamp {bInterpResult=true}
    boneToModify {boneName=base_backJacket}
    translation {z=2}
   }
   n30 = "Transform (Modify) Bone - Bone: base_backJacket_r"!(ComponentPose=@n29, bAlphaBoolEnabled:bool=True, Translation:Vector="0,0,2", Rotation="0,0,0") -> (Pose) {
    alphaInputType=Bool
    rotationMode=BMM_Additive
    rotationSpace=BCS_BoneSpace
    translationMode=BMM_Additive
    translationSpace=BCS_BoneSpace
    alphaBoolBlend {blendOption=Linear}
    alphaScaleBiasClamp {bInterpResult=true}
    boneToModify {boneName=base_backJacket_r}
    translation {z=2}
   }
   n31 = "Transform (Modify) Bone - Bone: base_midJacket_r"!(ComponentPose=@n30, bAlphaBoolEnabled:bool=True, Translation:Vector="0,0,0", Rotation="0,-5.671,0") -> (Pose) {
    alphaInputType=Bool
    rotationMode=BMM_Additive
    rotationSpace=BCS_BoneSpace
    translationMode=BMM_Additive
    translationSpace=BCS_BoneSpace
    alphaBoolBlend {blendOption=Linear}
    alphaScaleBiasClamp {bInterpResult=true}
    boneToModify {boneName=base_midJacket_r}
    rotation {yaw=-5.671}
   }
   n32 = "Transform (Modify) Bone - Bone: base_frontJacket_r"!(ComponentPose=@n31, bAlphaBoolEnabled:bool=True, Translation:Vector="0,0,0", Rotation="0,0,45") -> (Pose) {
    alphaInputType=Bool
    rotationMode=BMM_Additive
    rotationSpace=BCS_BoneSpace
    translationMode=BMM_Additive
    translationSpace=BCS_BoneSpace
    alphaBoolBlend {blendOption=Linear}
    alphaScaleBiasClamp {bInterpResult=true}
    boneToModify {boneName=base_frontJacket_r}
    rotation {roll=45}
   }
   n33 = "Transform (Modify) Bone - Bone: base_frontJacket_l"!(ComponentPose=@n32, bAlphaBoolEnabled=True, Translation="0,0,-2.432", Rotation="0,0,0", Scale:Vector="1,1,1") -> (Pose) {
    alphaInputType=Bool
    rotationMode=BMM_Additive
    rotationSpace=BCS_BoneSpace
    translationMode=BMM_Additive
    translationSpace=BCS_BoneSpace
    alphaBoolBlend {blendOption=Linear}
    alphaScaleBiasClamp {bInterpResult=true}
    boneToModify {boneName=base_frontJacket_l}
    translation {z=-2.432}
   }
   n34 = "Transform (Modify) Bone - Bone: base_frontJacket_r"!(ComponentPose=@n33, bAlphaBoolEnabled=True, Translation="0,0,1", Rotation="0,0,0", Scale:Vector="1,1,1") -> (Pose) {
    alphaInputType=Bool
    rotationMode=BMM_Additive
    rotationSpace=BCS_BoneSpace
    translationMode=BMM_Additive
    translationSpace=BCS_BoneSpace
    alphaBoolBlend {blendOption=Linear}
    alphaScaleBiasClamp {bInterpResult=true}
    boneToModify {boneName=base_frontJacket_r}
    translation {z=1}
   }
   Save cached pose 'presimFixesCP'(Pose=Component To Local(ComponentPose=@n34)) {cacheName=presimFixesCP}
  ]}
 EventGraph {nodes [
   Event BlueprintUpdateAnimation()
   > n1 = Sequence()
   n2 = NOT Boolean(A=Get bIsDBNO())
   n3 = NOT Boolean(A=Get bIsCrouching())
   n4 = NOT Boolean(A=Get bIsSkyDiving())
   n5 = OR Boolean(A=Get bIsSprinting(), B=Get bIsMoving2D())
   > Set isRunning(isRunning:double="To Float (Boolean)"(InBool=AND Boolean!(A=@n5, B=@n3, C:bool=@n4, D:bool=@n2) -> (ReturnValue))) -> (Output_Get:double)
   n6 = Get bIsSliding()
   n7 = Get bIsBaseballSliding()
   n8 = Get bIsKneeSliding()
   > Set isSliding(isSliding:double="To Float (Boolean)"(InBool=OR Boolean!(A=@n6, B=@n8, C:bool=@n7) -> (ReturnValue))) -> (Output_Get:double)
   n9 = Get bIsPlayingEmote()
   n10 = Get bIsSkyDiving()
   > Set isFrontEndAndNotEmote(isFrontEndAndNotEmote:bool=AND Boolean!(A=Get bIsInFrontEnd(), B=NOT Boolean(A=@n9), C=NOT Boolean(A=@n10)) -> (ReturnValue)) -> (Output_Get:bool)
   n11 = Get bIsSliding()
   > Set IsReallyCrouched(IsReallyCrouched:bool=AND Boolean!(A=Get bIsCrouching(), B=NOT Boolean(A=@n11)) -> (ReturnValue)) -> (Output_Get:bool)
   n12 = "Gravity Override State Function (Nativized)"!(
    execute=@n1.then_1
    CharacterPartAnimInstance:Custom Character Part Anim Instance Object Reference=Self-Reference()
    GravityOverrideParameterSet:Fort Gravity Override Parameter Set Structure=Get Gravity Override Params() -> (Gravity Override Params:Fort Gravity Override Parameter Set Structure)
    RigidBodyAnimNodeParameterSet:Fort Rigid Body Anim Node Parameter Set Structure=Get RBAN Params() -> (RBAN Params:Fort Rigid Body Anim Node Parameter Set Structure)
    DebugDraw:bool=false
    GameTimeInSeconds:float=GetGameTimeInSeconds()
    DebugJointName:Name=root
    StateLogicLayerName:Name=none
   ) -> (
    then
    OutGravityOverride:Vector
    OutCurrentState:Name=None
    OutGravityOverrideParameters:Fort Gravity Override Parameters Structure
    OutRigidBodyAnimNodeParameters:Fort Rigid Body Anim Node Parameters Structure
   )
   > Set Gravity Override(Gravity Override=@n12)
   > Set RBAN Settings Out(RBAN Settings Out=@n12.OutRigidBodyAnimNodeParameters)
  ]}
]
variables [
 Gravity Override {type=Vector}
 Gravity Override Params {
  type=FortGravityOverrideParameterSet
  default=(Skydive_GoingUp=(GravityOverride=(WindFrequency=(X=1.500000,Y=7.000000,Z=8.000000),WindAmplitude=(X=200.000000,Y=-500.000000,Z=400.000000),GravityOverride=(X=0.000000,Y=-300.000000,Z=1500.000000),JointName="root")),Skydive_FloatingDown=(GravityOverride=(WindFrequency=(X=1.500000,Y=7.000000,Z=8.000000),WindAmplitude=(X=200.000000,Y=-500.000000,Z=400.000000),WindOffset=(X=0.000000,Y=-600.000000,Z=400.000000),GravityOverride=(X=0.000000,Y=-1000.000000,Z=600.000000),JointName="root")),Skydive_DivingDown=(GravityOverride=(WindFrequency=(X=1.500000,Y=12.000000,Z=10.000000),WindAmplitude=(X=50.000000,Y=300.000000,Z=100.000000),GravityOverride=(X=0.000000,Y=-50.000000,Z=-50.000000),JointName="pelvis")),Skydive_Parachuting=(GravityOverride=(WindFrequency=(X=1.500000,Y=7.000000,Z=8.000000),WindAmplitude=(X=200.000000,Y=-500.000000,Z=400.000000),GravityOverride=(X=0.000000,Y=-800.000000,Z=-980.000000),JointName="root")),OnGround_Standing=(GravityOverride=(GravityOverride=(X=0.000000,Y=0.000000,Z=-980.000000),JointName="root")),OnGround_Moving=(GravityOverride=(GravityOverride=(X=0.000000,Y=-500.000000,Z=-980.000000),JointName="root")),OnGround_Moving_HighRotationRate=(GravityOverride=(GravityOverride=(X=0.000000,Y=0.000000,Z=-600.000000),JointName="root")),OnGround_Crouch_Moving=(GravityOverride=(GravityOverride=(X=0.000000,Y=0.000000,Z=-980.000000),JointName="root")),Emote_or_Melee=(GravityOverride=(GravityOverride=(X=0.000000,Y=0.000000,Z=-980.000000),JointName="root")),Swimming=(GravityOverride=(GravityOverride=(X=-800.000000,Y=0.000000,Z=0.000000),JointName="pelvis")),Front_End=(GravityOverride=(GravityOverride=(X=0.000000,Y=0.000000,Z=-980.000000),JointName="root")),DBNO=(GravityOverride=(GravityOverride=(X=-980.000000,Y=400.000000,Z=0.000000),JointName="pelvis")),DBNO_Carried=(GravityOverride=(GravityOverride=(X=-980.000000,Y=500.000000,Z=-500.000000),JointName="pelvis")),OnGround_Falling=(GravityOverride=(GravityOverride=(X=0.000000,Y=0.000000,Z=-980.000000),JointName="root")),Ziplining=(GravityOverride=(GravityOverride=(X=0.000000,Y=0.000000,Z=-980.000000),JointName="root")),Skydive_Surfing=(GravityOverride=(WindFrequency=(X=1.500000,Y=7.000000,Z=8.000000),WindAmplitude=(X=200.000000,Y=-500.000000,Z=400.000000),GravityOverride=(X=500.000000,Y=-300.000000,Z=-980.000000),JointName="root")),Riding_Standing=(GravityOverride=(GravityOverride=(X=0.000000,Y=-350.000000,Z=-980.000000),JointName="root")),Riding_Moving=(GravityOverride=(GravityOverride=(X=0.000000,Y=-450.000000,Z=-450.000000),JointName="root")),Riding_Moving_Targeting=(GravityOverride=(GravityOverride=(X=0.000000,Y=-200.000000,Z=-980.000000),JointName="root")),Tactical_Sprint=(GravityOverride=(GravityOverride=(X=0.000000,Y=-850.000000,Z=1000.000000),JointName="root")),Rail_Grinding=(GravityOverride=(GravityOverride=(X=0.000000,Y=0.000000,Z=-980.000000),JointName="root")),)
 }
 isFrontEndAndNotEmote {type=boolean}
 IsReallyCrouched {type=boolean}
 isRunning {type=number}
 isSliding {type=number}
 RBAN Params {
  type=FortRigidBodyAnimNodeParameterSet
  default=(Skydive_GoingUp=(RigidBodyAnimNode=(SimSpaceSettings=(WorldAlpha=1.000000,MaxLinearVelocity=500.000000,MaxAngularVelocity=3.000000,MaxLinearAcceleration=1000.000000,MaxAngularAcceleration=1.000000,ExternalLinearDragV=(X=1.000000,Y=1.000000,Z=1.000000)))),Skydive_FloatingDown=(RigidBodyAnimNode=(ComponentLinearAccScale=(X=0.000000,Y=0.500000,Z=0.000000),ComponentLinearVelScale=(X=0.000000,Y=0.500000,Z=0.000000),SimSpaceSettings=(WorldAlpha=1.000000,MaxLinearVelocity=500.000000,MaxAngularVelocity=3.000000,MaxLinearAcceleration=1000.000000,MaxAngularAcceleration=1.000000,ExternalLinearDragV=(X=1.000000,Y=1.000000,Z=1.000000)))),Skydive_DivingDown=(RigidBodyAnimNode=(ComponentLinearAccScale=(X=0.100000,Y=0.000000,Z=0.000000),ComponentLinearVelScale=(X=0.100000,Y=0.000000,Z=0.000000),SimSpaceSettings=(WorldAlpha=1.000000,MaxLinearVelocity=500.000000,MaxAngularVelocity=3.000000,MaxLinearAcceleration=1000.000000,MaxAngularAcceleration=1.000000,ExternalLinearDragV=(X=1.000000,Y=1.000000,Z=1.000000)))),Skydive_Parachuting=(RigidBodyAnimNode=(ComponentLinearAccScale=(X=0.000000,Y=0.100000,Z=0.300000),ComponentLinearVelScale=(X=0.000000,Y=0.100000,Z=0.300000),SimSpaceSettings=(WorldAlpha=1.000000,MaxLinearVelocity=500.000000,MaxAngularVelocity=3.000000,MaxLinearAcceleration=1000.000000,MaxAngularAcceleration=1.000000,ExternalLinearDragV=(X=1.000000,Y=1.000000,Z=1.000000)))),OnGround_Standing=(RigidBodyAnimNode=(ComponentLinearAccScale=(X=0.300000,Y=0.000000,Z=0.000000),ComponentLinearVelScale=(X=0.300000,Y=0.000000,Z=0.000000),SimSpaceSettings=(WorldAlpha=1.000000,DampingAlpha=0.250000,MaxLinearVelocity=500.000000,MaxAngularVelocity=20.000000,MaxLinearAcceleration=1000.000000,MaxAngularAcceleration=80.000000,ExternalLinearDragV=(X=1.000000,Y=1.000000,Z=1.000000)))),OnGround_Moving=(RigidBodyAnimNode=(ComponentLinearAccScale=(X=0.300000,Y=0.000000,Z=0.000000),ComponentLinearVelScale=(X=0.300000,Y=0.000000,Z=0.000000),SimSpaceSettings=(WorldAlpha=1.000000,DampingAlpha=0.250000,MaxLinearVelocity=500.000000,MaxAngularVelocity=20.000000,MaxLinearAcceleration=1000.000000,MaxAngularAcceleration=80.000000,ExternalLinearDragV=(X=1.000000,Y=1.000000,Z=1.000000)))),OnGround_Moving_HighRotationRate=(RigidBodyAnimNode=(ComponentLinearAccScale=(X=0.300000,Y=0.300000,Z=0.000000),ComponentLinearVelScale=(X=0.300000,Y=0.300000,Z=0.000000),SimSpaceSettings=(WorldAlpha=1.000000,DampingAlpha=0.500000,MaxLinearVelocity=500.000000,MaxAngularVelocity=3.000000,MaxLinearAcceleration=1000.000000,MaxAngularAcceleration=1.000000,ExternalLinearDragV=(X=1.000000,Y=1.000000,Z=1.000000)))),OnGround_Crouch_Moving=(RigidBodyAnimNode=(SimSpaceSettings=(WorldAlpha=1.000000,MaxLinearVelocity=500.000000,MaxAngularVelocity=3.000000,MaxLinearAcceleration=1000.000000,MaxAngularAcceleration=1.000000,ExternalLinearDragV=(X=1.000000,Y=1.000000,Z=1.000000)))),Emote_or_Melee=(RigidBodyAnimNode=(SimSpaceSettings=(WorldAlpha=1.000000,DampingAlpha=0.250000,MaxLinearVelocity=500.000000,MaxAngularVelocity=3.000000,MaxLinearAcceleration=1000.000000,MaxAngularAcceleration=1.000000,ExternalLinearDragV=(X=1.000000,Y=1.000000,Z=1.000000)))),Swimming=(RigidBodyAnimNode=(SimSpaceSettings=(WorldAlpha=1.000000,MaxLinearVelocity=500.000000,MaxAngularVelocity=3.000000,MaxLinearAcceleration=1000.000000,MaxAngularAcceleration=1.000000,ExternalLinearDragV=(X=1.000000,Y=1.000000,Z=1.000000)))),Front_End=(RigidBodyAnimNode=(SimSpaceSettings=(WorldAlpha=1.000000,MaxLinearVelocity=500.000000,MaxAngularVelocity=3.000000,MaxLinearAcceleration=1000.000000,MaxAngularAcceleration=1.000000,ExternalLinearDragV=(X=1.000000,Y=1.000000,Z=1.000000)))),DBNO=(RigidBodyAnimNode=(SimSpaceSettings=(WorldAlpha=1.000000,MaxLinearVelocity=500.000000,MaxAngularVelocity=3.000000,MaxLinearAcceleration=1000.000000,MaxAngularAcceleration=1.000000,ExternalLinearDragV=(X=1.000000,Y=1.000000,Z=1.000000)))),DBNO_Carried=(RigidBodyAnimNode=(SimSpaceSettings=(WorldAlpha=1.000000,MaxLinearVelocity=500.000000,MaxAngularVelocity=3.000000,MaxLinearAcceleration=1000.000000,MaxAngularAcceleration=1.000000,ExternalLinearDragV=(X=1.000000,Y=1.000000,Z=1.000000)))),OnGround_Falling=(RigidBodyAnimNode=(SimSpaceSettings=(WorldAlpha=1.000000,DampingAlpha=0.500000,MaxLinearVelocity=500.000000,MaxAngularVelocity=20.000000,MaxLinearAcceleration=1000.000000,MaxAngularAcceleration=80.000000,ExternalLinearDragV=(X=1.000000,Y=1.000000,Z=1.000000)))),Ziplining=(RigidBodyAnimNode=(SimSpaceSettings=(WorldAlpha=1.000000,MaxLinearVelocity=500.000000,MaxAngularVelocity=3.000000,MaxLinearAcceleration=1000.000000,MaxAngularAcceleration=1.000000,ExternalLinearDragV=(X=1.000000,Y=1.000000,Z=1.000000)))),Skydive_Surfing=(RigidBodyAnimNode=(SimSpaceSettings=(WorldAlpha=1.000000,MaxLinearVelocity=500.000000,MaxAngularVelocity=3.000000,MaxLinearAcceleration=1000.000000,MaxAngularAcceleration=1.000000,ExternalLinearDragV=(X=1.000000,Y=1.000000,Z=1.000000)))),Riding_Standing=(RigidBodyAnimNode=(SimSpaceSettings=(WorldAlpha=1.000000,MaxLinearVelocity=500.000000,MaxAngularVelocity=3.000000,MaxLinearAcceleration=1000.000000,MaxAngularAcceleration=1.000000,ExternalLinearDragV=(X=1.000000,Y=1.000000,Z=1.000000)))),Riding_Moving=(RigidBodyAnimNode=(SimSpaceSettings=(WorldAlpha=1.000000,MaxLinearVelocity=500.000000,MaxAngularVelocity=3.000000,MaxLinearAcceleration=1000.000000,MaxAngularAcceleration=1.000000,ExternalLinearDragV=(X=1.000000,Y=1.000000,Z=1.000000)))),Riding_Moving_Targeting=(RigidBodyAnimNode=(SimSpaceSettings=(WorldAlpha=1.000000,MaxLinearVelocity=500.000000,MaxAngularVelocity=3.000000,MaxLinearAcceleration=1000.000000,MaxAngularAcceleration=1.000000,ExternalLinearDragV=(X=1.000000,Y=1.000000,Z=1.000000)))),Tactical_Sprint=(RigidBodyAnimNode=(ComponentLinearAccScale=(X=0.300000,Y=0.000000,Z=0.000000),ComponentLinearVelScale=(X=0.300000,Y=0.000000,Z=0.000000),SimSpaceSettings=(WorldAlpha=1.000000,MaxLinearVelocity=500.000000,MaxAngularVelocity=3.000000,MaxLinearAcceleration=1000.000000,MaxAngularAcceleration=1.000000,ExternalLinearDragV=(X=1.000000,Y=1.000000,Z=1.000000)))),Rail_Grinding=(RigidBodyAnimNode=(SimSpaceSettings=(WorldAlpha=1.000000,MaxLinearVelocity=500.000000,MaxAngularVelocity=3.000000,MaxLinearAcceleration=1000.000000,MaxAngularAcceleration=1.000000,ExternalLinearDragV=(X=1.000000,Y=1.000000,Z=1.000000)))),)
 }
 RBAN Settings Out {type=FortRigidBodyAnimNodeParameters}
]
