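# Animation Blueprint for the CrawlyTech male body's face accessory, adding physics motion to its left and right antennae.
# It copies the attached parent's pose, then chains six root-relative Anim Dynamics nodes, three per antenna, with angular springs stiffer on the left.
# Zipline input events and a deprecated gravity helper update toggle and gravity variables, but the dynamics nodes show fixed enable and gravity values rather than reading them.
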AnimBlueprint /BRCosmetics/Characters/Player/Male/Medium/Bodies/M_MED_CrawlyTech/Meshes/Parts/M_MED_CrawlyTech_FaceAcc_AnimBP
target_skeleton=&/M_MED_CrawlyTech_FaceAcc_Skeleton
graphs [
 AnimGraph {nodes [
   n1 = Anim Dynamics - Bone: dyn_antenna_l_01!(
    ComponentPose=Local To Component(LocalPose=Copy Pose From Mesh() {bUseAttachedParent=true})
    bAlphaBoolEnabled:bool=True
    GravityOverride:Vector="0,0,-980"
    ExternalForce="0,0,0"
   ) -> (Pose) {
    alphaInputType=Bool
    angularDampingOverride=1
    angularSpringConstant=20000
    bAngularSpring=true
    bOverrideAngularDamping=true
    bUseGravityOverride=true
    gravityScale=0.35
    simulationSpace=RootRelative
    alphaBoolBlend {blendOption=Linear}
    boundBone {boneName=dyn_antenna_l_01}
    componentLinearAccScale=<1 1 1>
    componentLinearVelScale=<1 1 1>
    gravityOverride {z=-980}
    physicsBodyDefinitions [{
      boundBone {boneName=dyn_antenna_l_01}
      boxExtents=<20 30 30>
      constraintSetup {angularLimitsMax=<12 0 25>, angularLimitsMin=<-15 -15 -25>, angularTarget=<1 0 0>}
      localJointOffset=<6 0 0>
     }]
   }
   n2 = Anim Dynamics - Bone: dyn_antenna_l_02!(ComponentPose=@n1, bAlphaBoolEnabled:bool=True, GravityOverride:Vector="0,0,-980", ExternalForce="0,0,0") -> (Pose) {
    alphaInputType=Bool
    angularDampingOverride=1
    angularSpringConstant=20000
    bAngularSpring=true
    bOverrideAngularDamping=true
    bUseGravityOverride=true
    gravityScale=0.35
    simulationSpace=RootRelative
    alphaBoolBlend {blendOption=Linear}
    boundBone {boneName=dyn_antenna_l_02}
    componentLinearAccScale=<1 1 1>
    componentLinearVelScale=<1 1 1>
    gravityOverride {z=-980}
    physicsBodyDefinitions [{
      boundBone {boneName=dyn_antenna_l_02}
      boxExtents=<20 30 30>
      constraintSetup {angularLimitsMax=<15 0 25>, angularLimitsMin=<-15 -15 -25>, angularTarget=<1 0 0>}
      localJointOffset=<6 0 0>
     }]
   }
   n3 = Anim Dynamics - Bone: dyn_antenna_l_03!(ComponentPose=@n2, bAlphaBoolEnabled:bool=True, GravityOverride:Vector="0,0,-980", ExternalForce="0,0,0") -> (Pose) {
    alphaInputType=Bool
    angularDampingOverride=1
    angularSpringConstant=20000
    bAngularSpring=true
    bOverrideAngularDamping=true
    bUseGravityOverride=true
    gravityScale=0.35
    simulationSpace=RootRelative
    alphaBoolBlend {blendOption=Linear}
    boundBone {boneName=dyn_antenna_l_03}
    componentLinearAccScale=<1 1 1>
    componentLinearVelScale=<1 1 1>
    gravityOverride {z=-980}
    physicsBodyDefinitions [{
      boundBone {boneName=dyn_antenna_l_03}
      boxExtents=<20 30 30>
      constraintSetup {angularLimitsMax=<15 5 25>, angularLimitsMin=<-15 -15 -25>, angularTarget=<1 0 0>}
      localJointOffset=<6 0 0>
     }]
   }
   n4 = Anim Dynamics - Bone: dyn_antenna_r_01!(ComponentPose=@n3, bAlphaBoolEnabled:bool=True, GravityOverride:Vector="0,0,-980", ExternalForce="0,0,0") -> (Pose) {
    alphaInputType=Bool
    angularDampingOverride=1
    angularSpringConstant=10000
    bAngularSpring=true
    bOverrideAngularDamping=true
    bUseGravityOverride=true
    gravityScale=0.35
    simulationSpace=RootRelative
    alphaBoolBlend {blendOption=Linear}
    boundBone {boneName=dyn_antenna_r_01}
    componentLinearAccScale=<1 1 1>
    componentLinearVelScale=<1 1 1>
    gravityOverride {z=-980}
    physicsBodyDefinitions [{
      boundBone {boneName=dyn_antenna_r_01}
      boxExtents=<20 30 30>
      constraintSetup {angularLimitsMax=<12 0 15>, angularLimitsMin=<-15 -15 -15>, angularTarget=<1 0 0>}
      localJointOffset=<3 0 0>
     }]
   }
   n5 = Anim Dynamics - Bone: dyn_antenna_r_02!(ComponentPose=@n4, bAlphaBoolEnabled:bool=True, GravityOverride:Vector="0,0,-980", ExternalForce="0,0,0") -> (Pose) {
    alphaInputType=Bool
    angularDampingOverride=1
    angularSpringConstant=10000
    bAngularSpring=true
    bOverrideAngularDamping=true
    bUseGravityOverride=true
    gravityScale=0.35
    simulationSpace=RootRelative
    alphaBoolBlend {blendOption=Linear}
    boundBone {boneName=dyn_antenna_r_02}
    componentLinearAccScale=<1 1 1>
    componentLinearVelScale=<1 1 1>
    gravityOverride {z=-980}
    physicsBodyDefinitions [{
      boundBone {boneName=dyn_antenna_r_02}
      boxExtents=<20 30 30>
      constraintSetup {angularLimitsMax=<15 0 15>, angularLimitsMin=<-15 -15 -15>, angularTarget=<1 0 0>}
      localJointOffset=<3 0 0>
     }]
   }
   n6 = Anim Dynamics - Bone: dyn_antenna_r_03!(ComponentPose=@n5, bAlphaBoolEnabled:bool=True, GravityOverride:Vector="0,0,-980", ExternalForce="0,0,0") -> (Pose) {
    alphaInputType=Bool
    angularDampingOverride=1
    angularSpringConstant=10000
    bAngularSpring=true
    bOverrideAngularDamping=true
    bUseGravityOverride=true
    gravityScale=0.35
    simulationSpace=RootRelative
    alphaBoolBlend {blendOption=Linear}
    boundBone {boneName=dyn_antenna_r_03}
    componentLinearAccScale=<1 1 1>
    componentLinearVelScale=<1 1 1>
    gravityOverride {z=-980}
    physicsBodyDefinitions [{
      boundBone {boneName=dyn_antenna_r_03}
      boxExtents=<20 30 30>
      constraintSetup {angularLimitsMax=<15 5 15>, angularLimitsMin=<-15 -15 -15>, angularTarget=<1 0 0>}
      localJointOffset=<3 0 0>
     }]
   }
   Output Pose(Result=Component To Local(ComponentPose=@n6)) {layerGroup=None}
  ]}
 EventGraph {nodes [
   Event BlueprintUpdateAnimation()
   > n1 = Gravity Override State Function DEPRECATED(CharacterPartAnimInstance=Self-Reference(), GameTimeInSeconds=GetGameTimeInSeconds(), GravityOverrideParams=Get Gravity Override Params())
   > Set Gravity Override(Gravity Override=@n1)
   > Set toggleAnimDynamics(toggleAnimDynamics:bool=NOT Boolean(A=Get isOnZipline() -> (isOnZipline:bool))) -> (Output_Get:bool)
   Event BlueprintInitializeAnimation()
   > Set Skel Mesh(Skel Mesh=GetOwnerSkeletalMeshComponent())
   n2 = Event BlueprintProcessAnimInputEvent()
   > n3 = Branch(Condition=IsNamed(self=@n2.AnimInputEvent, InEventName=aie.Player.Zipline.Entered))
   > Set isOnZipline(isOnZipline:bool=1) -> (Output_Get:bool)
   Branch(execute=@n3.else, Condition=IsNamed(self=@n2.AnimInputEvent, InEventName=aie.Player.Zipline.Exited))
   > Set isOnZipline(isOnZipline=false)
  ]}
]
variables [
 Gravity Override {type=Vector}
 Gravity Override Params {type=GravityOverrideParamsStruct}
 isOnZipline {type=boolean}
 Skel Mesh {type=SkeletalMeshComponent}
 toggleAnimDynamics {type=boolean}
]
